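# Niagara system for short-lived, spinning glitter sprites on the female medium BrainMatter jacket body cosmetic.
# Six leg, arm and torso emitters have spawn rates gated by User bools for Default Jacket, Glow Jacket and Glow variants.
# Particles follow skeletal mesh bones each update using offset arrays, mirrored by partition on limbs.
# Size scales with camera distance, brightness with time of day, and white torso emitters vary camera offset by view direction.
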NiagaraSystem /BRCosmetics/Characters/Player/Female/Medium/Bodies/F_MED_BrainMatter_Jacket/FX/NS_MED_BrainMatter_Glitter
dependencies {
 usedDataInterfaces [&NiagaraDataInterfaceArrayInt32, &NiagaraDataInterfaceCurve, &NiagaraDataInterfaceCamera]
 usedDynamicInputs []
 usedModules [
  &self:SystemSpawnScript.SystemScriptSource.SystemScriptGraph.NiagaraNodeAssignment_0.SetVariables_#1
  &SystemState
  &self:SystemSpawnScript.SystemScriptSource.SystemScriptGraph.NiagaraNodeAssignment_1.SetVariables_#2
 ]
}
emitters {
 Leg {
  emitterUpdateScript [
   &EmitterState()
   &SpawnRate(
    Spawn Probability=$System.SpawnProbability
    SpawnRate=&Multiply_Float(
     A=$System.SpawnRate
     B=&MakeCustomFloatFromBool(
      Bool=&Boolean_OrOperation(Bool1=$User.GlowJacketVariantActive, Bool2=$User.GlowVariantActive)
      True Float=0.6
      False Float=0
     )
    )
    Use Spawn Probability=true
   )
  ]
  particleSpawnScript [
   &PartitionParticles(Particles Per Partition=1, Partition Mode:ENiagara_PartitionMode=::NewEnumerator1)
   &InitializeParticle(
    Color Mode=::NewEnumerator0
    Lifetime=&RandomRangeFloat(Minimum=0.3, Maximum=0.4)
    Sprite Rotation Mode=::NewEnumerator0
    Sprite Size Mode=::NewEnumerator1
    Uniform Sprite Size Max=$System.MaxSize
    Uniform Sprite Size Min=$System.MinSize
    UsePositionOffset=false
    Write Material Random=true
   )
   &CameraOffset(Camera Offset Amount=$System.CameraOffset)
  ]
  particleUpdateScript [
   &ParticleState()
   &SkeletalMeshLocation(
    Bone / Socket Index=&SelectIntFromArray(Direct Array Index=$Particles.PartitionParticles.Partition, Int Selection Array=DataInterface() {propertyValues {IntData [0,1]}})
    Bone Sampling Mode=::NewEnumerator4
    MaskSpawnGroup=false
    Sampled Position Offset=&Multiply_Vector(
     A=&SelectVectorFromArray(Array Sampling Mode=::NewEnumerator0, Direct Array Index=?, Vector Selection Array=$User.Array_Thigh, Random Seed=0)
     B=&MaskVectorBySpawnGroup(Masked Vector=<1 1 1>, Passthrough Vector=<-1 -1 -1>, Spawn Group Mask=$Particles.PartitionParticles.Partition)
    )
    Skeletal Mesh=$System.Skeletal Mesh 1
   ) {module=SkeletalMeshLocation004}
   &SpriteRotationRate(Rotation Rate=250)
   &ScaleSpriteSize(
    Scale Sprite Size Mode=::NewEnumerator0
    Uniform Scale Factor=&FloatFromCurve(
     FloatCurve=DataInterface() {propertyValues {Curve {keys [{arriveTangent=2.9879, interpMode=RCIM_Cubic, leaveTangent=2.9879, tangentMode=RCTM_User, time=0, value=0}, <at 1 1 RCIM_Cubic>]}}}
     Scale Curve=&FN_LOD_Distance_Remap(Start Distance=100, End Distance=1000, Minimum Output=1, Maximum Output=3, Clamp Output=true)
    )
   )
   &ScaleColor(
    Scale Alpha=&FloatFromCurve(FloatCurve=DataInterface() {propertyValues {Curve {keys [
         {arriveTangent=9.0725, interpMode=RCIM_Cubic, leaveTangent=9.0725, tangentMode=RCTM_User, time=0, value=0}
         {arriveTangent=-2.5249, interpMode=RCIM_Cubic, leaveTangent=-2.5249, tangentMode=RCTM_User, time=0.1408, value=0.9847}
         {arriveTangent=-0.0405, interpMode=RCIM_Cubic, leaveTangent=-0.0405, tangentMode=RCTM_User, time=1, value=0}
        ]}}})
    Scale RGB=&Multiply_VectorByFloat(
     Vector=<1 1 1>
     Float=&FNTimeOfDayScalar(Night Scale=0.2, Morning Scale=0.5, Day Scale=0.5, Evening Scale=0.5)
    )
    ScaleA=true
    ScaleRGB=true
   )
  ]
  rendererClasses [&NiagaraSpriteRendererProperties]
  renderers [{rendererClass=&NiagaraSpriteRendererProperties}]
 }
 LowerArm_Default {
  emitterUpdateScript [
   &EmitterState()
   &SpawnRate(
    Spawn Probability=$System.SpawnProbability
    SpawnRate=&Multiply_Float(
     A=$System.SpawnRate
     B=&MakeCustomFloatFromBool(Bool=$User.DefaultJacketVariantActive, True Float=0.6, False Float=0)
    )
    Use Spawn Probability=true
   )
  ]
  particleSpawnScript [
   &PartitionParticles(Particles Per Partition=1, Partition Mode:ENiagara_PartitionMode=::NewEnumerator1)
   &InitializeParticle(
    Color Mode=::NewEnumerator0
    Lifetime=&RandomRangeFloat(Minimum=0.3, Maximum=0.4)
    Sprite Rotation Mode=::NewEnumerator0
    Sprite Size Mode=::NewEnumerator1
    Uniform Sprite Size Max=$System.MaxSize
    Uniform Sprite Size Min=$System.MinSize
    UsePositionOffset=false
    Write Material Random=true
   )
   &CameraOffset(Camera Offset Amount=$System.CameraOffset)
  ]
  particleUpdateScript [
   &ParticleState()
   &SkeletalMeshLocation(
    Bone / Socket Index=&SelectIntFromArray(Direct Array Index=$Particles.PartitionParticles.Partition, Int Selection Array=DataInterface() {propertyValues {IntData [3,4]}})
    Bone Sampling Mode=::NewEnumerator4
    MaskSpawnGroup=false
    Sampled Position Offset=&Multiply_Vector(
     A=&SelectVectorFromArray(Array Sampling Mode=::NewEnumerator0, Direct Array Index=?, Vector Selection Array=$User.Array_LowerArm, Random Seed=0)
     B=&MaskVectorBySpawnGroup(Masked Vector=<1 1 1>, Passthrough Vector=<-1 -1 -1>, Spawn Group Mask=$Particles.PartitionParticles.Partition)
    )
    Skeletal Mesh=$System.Skeletal Mesh 2
   ) {module=SkeletalMeshLocation004}
   &SpriteRotationRate(Rotation Rate=250)
   &ScaleSpriteSize(
    Scale Sprite Size Mode=::NewEnumerator0
    Uniform Scale Factor=&FloatFromCurve(
     FloatCurve=DataInterface() {propertyValues {Curve {keys [{arriveTangent=2.9879, interpMode=RCIM_Cubic, leaveTangent=2.9879, tangentMode=RCTM_User, time=0, value=0}, <at 1 1 RCIM_Cubic>]}}}
     Scale Curve=&FN_LOD_Distance_Remap(Start Distance=100, End Distance=500, Minimum Output=0.75, Maximum Output=1.25, Clamp Output=true)
    )
   )
   &ScaleColor(
    Scale Alpha=&FloatFromCurve(FloatCurve=DataInterface() {propertyValues {Curve {keys [
         {arriveTangent=9.0725, interpMode=RCIM_Cubic, leaveTangent=9.0725, tangentMode=RCTM_User, time=0, value=0}
         {arriveTangent=-2.5249, interpMode=RCIM_Cubic, leaveTangent=-2.5249, tangentMode=RCTM_User, time=0.1408, value=0.9847}
         {arriveTangent=-0.0405, interpMode=RCIM_Cubic, leaveTangent=-0.0405, tangentMode=RCTM_User, time=1, value=0}
        ]}}})
    Scale RGB=&Multiply_VectorByFloat(
     Vector=<1 1 1>
     Float=&FNTimeOfDayScalar(Night Scale=0.2, Morning Scale=0.5, Day Scale=0.5, Evening Scale=0.5)
    )
    ScaleA=true
    ScaleRGB=true
   )
  ]
  rendererClasses [&NiagaraSpriteRendererProperties]
  renderers [{rendererClass=&NiagaraSpriteRendererProperties}]
 }
 Top_Red {
  emitterUpdateScript [
   &EmitterState()
   &SpawnRate(
    Spawn Probability=$System.SpawnProbability
    SpawnRate=&Multiply_Float(
     A=$System.SpawnRate
     B=&MakeCustomFloatFromBool(Bool=$User.GlowVariantActive, True Float=0.75, False Float=0)
    )
    Use Spawn Probability=true
   )
  ]
  particleSpawnScript [
   &PartitionParticles(Number Of Partitions=1, Particles Per Partition=1, Partition Mode:ENiagara_PartitionMode=::NewEnumerator1)
   &InitializeParticle(
    Color=<rgba 1 0 0 1>
    Lifetime=&RandomRangeFloat(Minimum=0.3, Maximum=0.4)
    Position Mode=::NewEnumerator0
    Sprite Rotation Mode=::NewEnumerator0
    Sprite Size Mode=::NewEnumerator1
    Uniform Sprite Size Max=$System.MaxSize
    Uniform Sprite Size Min=$System.MinSize
    Write Material Random=true
   )
   &CameraOffset(Camera Offset Amount=3)
  ]
  particleUpdateScript [
   &ParticleState()
   &SkeletalMeshLocation(
    Bone / Socket Index=5
    Bone Sampling Mode=::NewEnumerator4
    MaskSpawnGroup=false
    Sampled Position Offset=&SelectVectorFromArray(
     Array Sampling Mode=::NewEnumerator0
     Direct Array Index=?
     Vector Selection Array=DataInterface() {propertyValues {FloatData [<-14.8507 -7.7646 8.1616>, <-13.6986 7 10>, <-7.0835 1.7114 13.2>, <-11.6836 0 13.5>, <-10 -7 -13>, <-15 -2 -13>, <-7 6 -13>, <-11 9 -10>]}}
     Random Seed=0
    )
    Skeletal Mesh=$System.Skeletal Mesh 1
   ) {module=SkeletalMeshLocation004}
   &SpriteRotationRate(Rotation Rate=250)
   &ScaleSpriteSize(
    Scale Sprite Size Mode=::NewEnumerator0
    Uniform Scale Factor=&FloatFromCurve(
     FloatCurve=DataInterface() {propertyValues {Curve {keys [{arriveTangent=2.9879, interpMode=RCIM_Cubic, leaveTangent=2.9879, tangentMode=RCTM_User, time=0, value=0}, <at 1 1 RCIM_Cubic>]}}}
     Scale Curve=&FN_LOD_Distance_Remap(Start Distance=100, End Distance=500, Minimum Output=0.75, Maximum Output=1.25, Clamp Output=true)
    )
   )
   &ScaleColor(
    Scale Alpha=&FloatFromCurve(FloatCurve=DataInterface() {propertyValues {Curve {keys [
         {arriveTangent=9.0725, interpMode=RCIM_Cubic, leaveTangent=9.0725, tangentMode=RCTM_User, time=0, value=0}
         {arriveTangent=-2.5249, interpMode=RCIM_Cubic, leaveTangent=-2.5249, tangentMode=RCTM_User, time=0.1408, value=0.9847}
         {arriveTangent=-0.0405, interpMode=RCIM_Cubic, leaveTangent=-0.0405, tangentMode=RCTM_User, time=1, value=0}
        ]}}})
    Scale RGB=&Multiply_VectorByFloat(
     Vector=<1 1 1>
     Float=&FNTimeOfDayScalar(Night Scale=0.2, Morning Scale=0.5, Day Scale=0.5, Evening Scale=0.5)
    )
    ScaleA=true
    ScaleRGB=true
   )
  ]
  rendererClasses [&NiagaraSpriteRendererProperties]
  renderers [{rendererClass=&NiagaraSpriteRendererProperties}]
 }
 Top_White {
  emitterUpdateScript [
   &EmitterState()
   &SpawnRate(
    Spawn Probability=$System.SpawnProbability
    SpawnRate=&Multiply_Float(
     A=$System.SpawnRate
     B=&MakeCustomFloatFromBool(Bool=$User.GlowJacketVariantActive, True Float=0.6, False Float=0)
    )
    Use Spawn Probability=true
   )
  ]
  particleSpawnScript [
   &PartitionParticles(Number Of Partitions=6, Particles Per Partition=1, Partition Mode:ENiagara_PartitionMode=::NewEnumerator1)
   &InitializeParticle(
    Color Mode=::NewEnumerator0
    Lifetime=&RandomRangeFloat(Minimum=0.3, Maximum=0.4)
    Position Mode=::NewEnumerator0
    Sprite Rotation Mode=::NewEnumerator0
    Sprite Size Mode=::NewEnumerator1
    Uniform Sprite Size Max=$System.MaxSize
    Uniform Sprite Size Min=$System.MinSize
    Write Material Random=true
   )
  ]
  particleUpdateScript [
   &ParticleState()
   &SkeletalMeshLocation(
    Bone / Socket Index=5
    Bone Sampling Mode=::NewEnumerator4
    MaskSpawnGroup=false
    Sampled Position Offset=&SelectVectorFromArray(
     Direct Array Index=$Particles.Initial.PartitionParticles.Partition
     Vector Selection Array=DataInterface() {propertyValues {FloatData [<-6 6 -14>, <-16 -4 -14>, <-16 9 -12>, <-6 5 15.4>, <-16 -4 14>, <-18 9 13.5>]}}
    )
    Skeletal Mesh=$System.Skeletal Mesh 1
   ) {module=SkeletalMeshLocation004}
   &CameraOffset(Camera Offset Amount=&Multiply_Float(
     A=&MakeCustomFloatFromBool(
      Bool=&SetBoolByIntComparison(A=3, B=$Particles.PartitionParticles.Partition, Comparison Type=::NewEnumerator0)
      True Float=3
      False Float=-3
     )
     B=&Lerp_Float(
      A=1
      B=-1
      Alpha=&DotProduct(
       Vector A=&NormalizeVector(
        Vector To Normalize=&ReturnCameraVector(Camera Vectors=::NewEnumerator1, Coordinate Space=::World)
        Safe Normalize=true
        Safe Fallback Threshold=0
        Safe Fallback Vector=<0 0 0.001>
       )
       Vector B=&NormalizeVector(
        Vector To Normalize=&TransformVector(Vector=<1 0 0>, Destination Space=::World)
        Safe Normalize=true
        Safe Fallback Threshold=0
        Safe Fallback Vector=<0 0 0.001>
       )
      )
     )
    ))
   &SpriteRotationRate(Rotation Rate=250)
   &ScaleSpriteSize(
    Scale Sprite Size Mode=::NewEnumerator0
    Uniform Scale Factor=&FloatFromCurve(
     FloatCurve=DataInterface() {propertyValues {Curve {keys [{arriveTangent=2.9879, interpMode=RCIM_Cubic, leaveTangent=2.9879, tangentMode=RCTM_User, time=0, value=0}, <at 1 1 RCIM_Cubic>]}}}
     Scale Curve=&FN_LOD_Distance_Remap(Start Distance=100, End Distance=500, Minimum Output=0.75, Maximum Output=1.25, Clamp Output=true)
    )
   )
   &ScaleColor(
    Scale Alpha=&FloatFromCurve(FloatCurve=DataInterface() {propertyValues {Curve {keys [
         {arriveTangent=9.0725, interpMode=RCIM_Cubic, leaveTangent=9.0725, tangentMode=RCTM_User, time=0, value=0}
         {arriveTangent=-2.5249, interpMode=RCIM_Cubic, leaveTangent=-2.5249, tangentMode=RCTM_User, time=0.1408, value=0.9847}
         {arriveTangent=-0.0405, interpMode=RCIM_Cubic, leaveTangent=-0.0405, tangentMode=RCTM_User, time=1, value=0}
        ]}}})
    Scale RGB=&Multiply_VectorByFloat(
     Vector=<1 1 1>
     Float=&FNTimeOfDayScalar(Night Scale=0.2, Morning Scale=0.5, Day Scale=0.5, Evening Scale=0.5)
    )
    ScaleA=true
    ScaleRGB=true
   )
  ]
  rendererClasses [&NiagaraSpriteRendererProperties]
  renderers [{rendererClass=&NiagaraSpriteRendererProperties}]
 }
 Top_White_Default {
  emitterUpdateScript [
   &EmitterState()
   &SpawnRate(
    Spawn Probability=$System.SpawnProbability
    SpawnRate=&Multiply_Float(
     A=$System.SpawnRate
     B=&MakeCustomFloatFromBool(Bool=$User.DefaultJacketVariantActive, True Float=1, False Float=0)
    )
    Use Spawn Probability=true
   )
  ]
  particleSpawnScript [
   &PartitionParticles(Number Of Partitions=12, Particles Per Partition=1, Partition Mode:ENiagara_PartitionMode=::NewEnumerator1)
   &InitializeParticle(
    Color Mode=::NewEnumerator0
    Lifetime=&RandomRangeFloat(Minimum=0.3, Maximum=0.4)
    Position Mode=::NewEnumerator0
    Sprite Rotation Mode=::NewEnumerator0
    Sprite Size Mode=::NewEnumerator1
    Uniform Sprite Size Max=$System.MaxSize
    Uniform Sprite Size Min=$System.MinSize
    Write Material Random=true
   )
  ]
  particleUpdateScript [
   &ParticleState()
   &SkeletalMeshLocation(
    Bone Sampling Mode=::NewEnumerator4
    MaskSpawnGroup=false
    Sampled Position Offset=&SelectVectorFromArray(
     Direct Array Index=$Particles.Initial.PartitionParticles.Partition
     Vector Selection Array=DataInterface() {propertyValues {FloatData [
        <-6 6 -14>
        <-16 -4 -14>
        <-16 8 -12>
        <5 3.5 -14>
        <6 -8 -14>
        <9.5 -2 -15>
        <-6 5 14>
        <-16 -4 14>
        <-16 8 12>
        <6 -8 14>
        <5 3.5 14>
        <9.5 -2 15>
       ]}}
    )
    Skeletal Mesh=$System.Skeletal Mesh 2
   ) {module=SkeletalMeshLocation004}
   &CameraOffset(Camera Offset Amount=&Multiply_Float(
     A=&MakeCustomFloatFromBool(
      Bool=&SetBoolByIntComparison(A=5, B=$Particles.PartitionParticles.Partition, Comparison Type=::NewEnumerator0)
      True Float=3
      False Float=-3
     )
     B=&Lerp_Float(
      A=1
      B=-1
      Alpha=&DotProduct(
       Vector A=&NormalizeVector(
        Vector To Normalize=&ReturnCameraVector(Camera Vectors=::NewEnumerator1, Coordinate Space=::World)
        Safe Normalize=true
        Safe Fallback Threshold=0
        Safe Fallback Vector=<0 0 0.001>
       )
       Vector B=&NormalizeVector(
        Vector To Normalize=&TransformVector(Vector=<1 0 0>, Destination Space=::World)
        Safe Normalize=true
        Safe Fallback Threshold=0
        Safe Fallback Vector=<0 0 0.001>
       )
      )
     )
    ))
   &SpriteRotationRate(Rotation Rate=250)
   &ScaleSpriteSize(
    Scale Sprite Size Mode=::NewEnumerator0
    Uniform Scale Factor=&FloatFromCurve(
     FloatCurve=DataInterface() {propertyValues {Curve {keys [{arriveTangent=2.9879, interpMode=RCIM_Cubic, leaveTangent=2.9879, tangentMode=RCTM_User, time=0, value=0}, <at 1 1 RCIM_Cubic>]}}}
     Scale Curve=&FN_LOD_Distance_Remap(Start Distance=100, End Distance=500, Minimum Output=0.75, Maximum Output=1.25, Clamp Output=true)
    )
   )
   &ScaleColor(
    Scale Alpha=&FloatFromCurve(FloatCurve=DataInterface() {propertyValues {Curve {keys [
         {arriveTangent=9.0725, interpMode=RCIM_Cubic, leaveTangent=9.0725, tangentMode=RCTM_User, time=0, value=0}
         {arriveTangent=-2.5249, interpMode=RCIM_Cubic, leaveTangent=-2.5249, tangentMode=RCTM_User, time=0.1408, value=0.9847}
         {arriveTangent=-0.0405, interpMode=RCIM_Cubic, leaveTangent=-0.0405, tangentMode=RCTM_User, time=1, value=0}
        ]}}})
    Scale RGB=&Multiply_VectorByFloat(
     Vector=<1 1 1>
     Float=&FNTimeOfDayScalar(Night Scale=0.2, Morning Scale=0.5, Day Scale=0.5, Evening Scale=0.5)
    )
    ScaleA=true
    ScaleRGB=true
   )
  ]
  rendererClasses [&NiagaraSpriteRendererProperties]
  renderers [{rendererClass=&NiagaraSpriteRendererProperties}]
 }
 UpperArm_Default {
  emitterUpdateScript [
   &EmitterState()
   &SpawnRate(
    Spawn Probability=$System.SpawnProbability
    SpawnRate=&Multiply_Float(
     A=$System.SpawnRate
     B=&MakeCustomFloatFromBool(Bool=$User.DefaultJacketVariantActive, True Float=0.6, False Float=0)
    )
    Use Spawn Probability=true
   )
  ]
  particleSpawnScript [
   &PartitionParticles(Particles Per Partition=1, Partition Mode:ENiagara_PartitionMode=::NewEnumerator1)
   &InitializeParticle(
    Color Mode=::NewEnumerator0
    Lifetime=&RandomRangeFloat(Minimum=0.3, Maximum=0.4)
    Sprite Rotation Mode=::NewEnumerator0
    Sprite Size Mode=::NewEnumerator1
    Uniform Sprite Size Max=$System.MaxSize
    Uniform Sprite Size Min=$System.MinSize
    UsePositionOffset=false
    Write Material Random=true
   )
   &CameraOffset(Camera Offset Amount=$System.CameraOffset)
  ]
  particleUpdateScript [
   &ParticleState()
   &SkeletalMeshLocation(
    Bone / Socket Index=&SelectIntFromArray(Direct Array Index=$Particles.PartitionParticles.Partition, Int Selection Array=DataInterface() {propertyValues {IntData [1,2]}})
    Bone Sampling Mode=::NewEnumerator4
    MaskSpawnGroup=false
    Sampled Position Offset=&Multiply_Vector(
     A=&SelectVectorFromArray(Array Sampling Mode=::NewEnumerator0, Direct Array Index=?, Vector Selection Array=$User.Array_UpperArm, Random Seed=0)
     B=&MaskVectorBySpawnGroup(Masked Vector=<1 1 1>, Passthrough Vector=<-1 -1 -1>, Spawn Group Mask=$Particles.PartitionParticles.Partition)
    )
    Skeletal Mesh=$System.Skeletal Mesh 2
   ) {module=SkeletalMeshLocation004}
   &SpriteRotationRate(Rotation Rate=250)
   &ScaleSpriteSize(
    Scale Sprite Size Mode=::NewEnumerator0
    Uniform Scale Factor=&FloatFromCurve(
     FloatCurve=DataInterface() {propertyValues {Curve {keys [{arriveTangent=2.9879, interpMode=RCIM_Cubic, leaveTangent=2.9879, tangentMode=RCTM_User, time=0, value=0}, <at 1 1 RCIM_Cubic>]}}}
     Scale Curve=&FN_LOD_Distance_Remap(Start Distance=100, End Distance=500, Minimum Output=0.75, Maximum Output=1.25, Clamp Output=true)
    )
   )
   &ScaleColor(
    Scale Alpha=&FloatFromCurve(FloatCurve=DataInterface() {propertyValues {Curve {keys [
         {arriveTangent=9.0725, interpMode=RCIM_Cubic, leaveTangent=9.0725, tangentMode=RCTM_User, time=0, value=0}
         {arriveTangent=-2.5249, interpMode=RCIM_Cubic, leaveTangent=-2.5249, tangentMode=RCTM_User, time=0.1408, value=0.9847}
         {arriveTangent=-0.0405, interpMode=RCIM_Cubic, leaveTangent=-0.0405, tangentMode=RCTM_User, time=1, value=0}
        ]}}})
    Scale RGB=&Multiply_VectorByFloat(
     Vector=<1 1 1>
     Float=&FNTimeOfDayScalar(Night Scale=0.2, Morning Scale=0.5, Day Scale=0.5, Evening Scale=0.5)
    )
    ScaleA=true
    ScaleRGB=true
   )
  ]
  rendererClasses [&NiagaraSpriteRendererProperties]
  renderers [{rendererClass=&NiagaraSpriteRendererProperties}]
 }
}
summary {
 systemName=NS_MED_BrainMatter_Glitter
 emitters [
  {emitterName=Leg, rendererClasses [&NiagaraSpriteRendererProperties]}
  {emitterName=Top_Red, rendererClasses [&NiagaraSpriteRendererProperties]}
  {emitterName=Top_White, rendererClasses [&NiagaraSpriteRendererProperties]}
  {emitterName=Top_White_Default, rendererClasses [&NiagaraSpriteRendererProperties]}
  {emitterName=UpperArm_Default, rendererClasses [&NiagaraSpriteRendererProperties]}
  {emitterName=LowerArm_Default, rendererClasses [&NiagaraSpriteRendererProperties]}
 ]
 userVariables [
  User.Array_LowerArm {
   type=NiagaraDataInterfaceArrayFloat3
   description=""
   defaultValue {
    struct=NiagaraExt_VariableValue_DataInterface
    value {
     dataInterface=&self:NiagaraDataInterfaceArrayFloat3_3
     dataInterfaceClass=&NiagaraDataInterfaceArrayFloat3
    }
   }
  }
  User.Array_Thigh {
   type=NiagaraDataInterfaceArrayFloat3
   description=""
   defaultValue {
    struct=NiagaraExt_VariableValue_DataInterface
    value {
     dataInterface=&self:NiagaraDataInterfaceArrayFloat3_0
     dataInterfaceClass=&NiagaraDataInterfaceArrayFloat3
    }
   }
  }
  User.Array_UpperArm {
   type=NiagaraDataInterfaceArrayFloat3
   description=""
   defaultValue {
    struct=NiagaraExt_VariableValue_DataInterface
    value {
     dataInterface=&self:NiagaraDataInterfaceArrayFloat3_2
     dataInterfaceClass=&NiagaraDataInterfaceArrayFloat3
    }
   }
  }
  User.DefaultJacketVariantActive {type=NiagaraBool}
  User.GlowJacketVariantActive {type=NiagaraBool}
  User.GlowVariantActive {type=NiagaraBool}
  User.Skeletal Mesh 1 {
   type=NiagaraDataInterfaceSkeletalMesh
   defaultValue {
    struct=NiagaraExt_VariableValue_DataInterface
    value {dataInterface=&self:NiagaraDataInterfaceSkeletalMesh_0}
   }
  }
  User.Skeletal Mesh 2 {
   type=NiagaraDataInterfaceSkeletalMesh
   defaultValue {
    struct=NiagaraExt_VariableValue_DataInterface
    value {dataInterface=&self:NiagaraDataInterfaceSkeletalMesh_1}
   }
  }
 ]
}
