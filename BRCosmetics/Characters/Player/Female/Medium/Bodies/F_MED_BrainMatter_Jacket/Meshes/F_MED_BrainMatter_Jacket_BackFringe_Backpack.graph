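# Physics asset for the dangling back fringe on the female medium BrainMatter jacket, named as the Backpack variant.
# Five two-segment capsule chains (fringe 02, 04, 07, 10, 12) hang from small base bodies that are rotationally locked to spine_05.
# Strand joints allow only narrow swing and twist, with twist limits varying per strand.
# Large capsules on spine_02, 03 and 05 model the torso, linked by angularly free constraints.
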PhysicsAsset /BRCosmetics/Characters/Player/Female/Medium/Bodies/F_MED_BrainMatter_Jacket/Meshes/F_MED_BrainMatter_Jacket_BackFringe_Backpack
bodies {
 dyn_back_fringe_01_Base [{shapeType=Capsule, length=0.505, radius=0.505, shapeName=dyn_back_fringe_01_Base_capsule, rotation=<rot 0 -90 0>}]
 dyn_back_fringe_02_01 [{shapeType=Capsule, length=6, radius=1.3, shapeName=dyn_back_fringe_02_01_capsule, center=<3.2799 -1 -0.5>, rotation=<rot 90 0 0>}]
 dyn_back_fringe_02_02 [{shapeType=Capsule, length=6, radius=1.3, shapeName=dyn_back_fringe_02_02_capsule, center=<2.016 -1 -0.5>, rotation=<rot 0 90.0454 -89.3976>}]
 dyn_back_fringe_02_base [{shapeType=Capsule, length=0.505, radius=0.505, shapeName=dyn_back_fringe_02_base_capsule, rotation=<rot 0 -90 0>}]
 dyn_back_fringe_03_base [{shapeType=Capsule, length=0.505, radius=0.505, shapeName=dyn_back_fringe_03_base_capsule, rotation=<rot 0 -90 0>}]
 dyn_back_fringe_04_01 [{shapeType=Capsule, length=7, radius=1.3, shapeName=dyn_back_fringe_04_01_capsule, center=<3.2721 0 -0.5>, rotation=<rot 90 0 0>}]
 dyn_back_fringe_04_02 [{shapeType=Capsule, length=6, radius=1.3, shapeName=dyn_back_fringe_04_02_capsule, center=<2.2866 0 -0.5>, rotation=<rot 90 0 0>}]
 dyn_back_fringe_04_base [{shapeType=Capsule, length=0.505, radius=0.505, shapeName=dyn_back_fringe_04_base_capsule, rotation=<rot 0 -90 0>}]
 dyn_back_fringe_07_01 [{shapeType=Capsule, length=7, radius=1.3, shapeName=dyn_back_fringe_07_01_capsule, center=<2.9487 0 -0.5>, rotation=<rot 90 0 0>}]
 dyn_back_fringe_07_02 [{shapeType=Capsule, length=6, radius=1.3, shapeName=dyn_back_fringe_07_02_capsule, center=<1.8185 0 -0.5>, rotation=<rot 90 0 0>}]
 dyn_back_fringe_10_01 [{shapeType=Capsule, length=7, radius=1.3, shapeName=dyn_back_fringe_10_01_capsule, center=<3.1233 0 -0.5>, rotation=<rot 90 0 0>}]
 dyn_back_fringe_10_02 [{shapeType=Capsule, length=6, radius=1.3, shapeName=dyn_back_fringe_10_02_capsule, center=<1.9604 0 -0.5>, rotation=<rot 90 0 0>}]
 dyn_back_fringe_12_01 [{shapeType=Capsule, length=7, radius=1.3, shapeName=dyn_back_fringe_12_01_capsule, center=<3.6679 1 -0.5>, rotation=<rot 90 0 0>}]
 dyn_back_fringe_12_02 [{shapeType=Capsule, length=6, radius=1.3, shapeName=dyn_back_fringe_12_02_capsule, center=<2.831 1 -0.5>, rotation=<rot 90 0 0>}]
 dyn_back_fringe_5_base [{shapeType=Capsule, length=0.505, radius=0.505, shapeName=dyn_back_fringe_5_base_capsule, rotation=<rot 0 -90 0>}]
 spine_02 [{shapeType=Capsule, length=12.4239, radius=11, shapeName=spine_02_capsule, center=<0 0.3475 0.0071>, rotation=<rot -1.1765 -90 -0.0367>}]
 spine_03 [{shapeType=Capsule, length=10.3338, radius=11, shapeName=spine_03_capsule, center=<1.4538 -0.9376 0>, rotation=<rot 0 -90 0>}]
 spine_05 [{
   shapeType=Capsule
   length=15.8429
   radius=10.7441
   shapeName=spine_05_capsule
   center=<-4.9674 -0.3976 -0.5897>
   rotation=<rot 0 -86.7796 0.8917>
  }]
}
constraints [
 {
  bone1Name=spine_05
  bone2Name=spine_03
  swing1LimitDegrees=45
  swing1Motion=Free
  swing2LimitDegrees=45
  swing2Motion=Free
  twistLimitDegrees=45
  twistMotion=Free
 }
 {
  bone1Name=spine_03
  bone2Name=spine_02
  swing1LimitDegrees=45
  swing1Motion=Free
  swing2LimitDegrees=45
  swing2Motion=Free
  twistLimitDegrees=45
  twistMotion=Free
 }
 {
  bone1Name=dyn_back_fringe_01_Base
  bone2Name=spine_05
  swing1LimitDegrees=45
  swing1Motion=Locked
  swing2LimitDegrees=45
  swing2Motion=Locked
  twistLimitDegrees=45
  twistMotion=Locked
 }
 {
  bone1Name=dyn_back_fringe_02_base
  bone2Name=spine_05
  swing1LimitDegrees=45
  swing1Motion=Locked
  swing2LimitDegrees=45
  swing2Motion=Locked
  twistLimitDegrees=45
  twistMotion=Locked
 }
 {
  bone1Name=dyn_back_fringe_03_base
  bone2Name=spine_05
  swing1LimitDegrees=45
  swing1Motion=Locked
  swing2LimitDegrees=45
  swing2Motion=Locked
  twistLimitDegrees=45
  twistMotion=Locked
 }
 {
  bone1Name=dyn_back_fringe_04_base
  bone2Name=spine_05
  swing1LimitDegrees=45
  swing1Motion=Locked
  swing2LimitDegrees=45
  swing2Motion=Locked
  twistLimitDegrees=45
  twistMotion=Locked
 }
 {
  bone1Name=dyn_back_fringe_5_base
  bone2Name=spine_05
  swing1LimitDegrees=45
  swing1Motion=Locked
  swing2LimitDegrees=45
  swing2Motion=Locked
  twistLimitDegrees=45
  twistMotion=Locked
 }
 {bone1Name=dyn_back_fringe_10_01, bone2Name=dyn_back_fringe_04_base, swing1LimitDegrees=15, swing2LimitDegrees=2, twistLimitDegrees=5}
 {bone1Name=dyn_back_fringe_10_02, bone2Name=dyn_back_fringe_10_01, swing1LimitDegrees=15, swing2LimitDegrees=4, twistLimitDegrees=2}
 {bone1Name=dyn_back_fringe_07_02, bone2Name=dyn_back_fringe_07_01, swing1LimitDegrees=15, swing2LimitDegrees=4, twistLimitDegrees=0}
 {bone1Name=dyn_back_fringe_07_01, bone2Name=dyn_back_fringe_03_base, swing1LimitDegrees=15, swing2LimitDegrees=2, twistLimitDegrees=0}
 {bone1Name=dyn_back_fringe_04_02, bone2Name=dyn_back_fringe_04_01, swing1LimitDegrees=15, swing2LimitDegrees=4, twistLimitDegrees=0}
 {bone1Name=dyn_back_fringe_04_01, bone2Name=dyn_back_fringe_02_base, swing1LimitDegrees=15, swing2LimitDegrees=2, twistLimitDegrees=0}
 {bone1Name=dyn_back_fringe_12_01, bone2Name=dyn_back_fringe_5_base, swing1LimitDegrees=5, swing2LimitDegrees=2, twistLimitDegrees=4}
 {bone1Name=dyn_back_fringe_12_02, bone2Name=dyn_back_fringe_12_01, swing1LimitDegrees=15, swing2LimitDegrees=4, twistLimitDegrees=4}
 {bone1Name=dyn_back_fringe_02_01, bone2Name=dyn_back_fringe_01_Base, swing1LimitDegrees=15, swing2LimitDegrees=2, twistLimitDegrees=10}
 {bone1Name=dyn_back_fringe_02_02, bone2Name=dyn_back_fringe_02_01, swing1LimitDegrees=15, swing2LimitDegrees=4, twistLimitDegrees=10}
]
