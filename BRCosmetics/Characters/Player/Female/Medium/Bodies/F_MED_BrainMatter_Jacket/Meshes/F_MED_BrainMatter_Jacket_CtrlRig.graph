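# Control Rig for the female BrainMatter jacket cosmetic, adding pose-driven corrections and secondary motion in three sequenced passes.
# Upper-arm spherical pose readers weight collar corrective offsets.
# Back fringe strands follow neighboring strands through spring-interpolated global rotations, with partial single-axis position constraints on the outer pairs.
# Front jacket base bones are offset by thigh and spine pose readers, plus remapped Euler angles of spine_05 relative to spine_01.
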ControlRigBlueprint /BRCosmetics/Characters/Player/Female/Medium/Bodies/F_MED_BrainMatter_Jacket/Meshes/F_MED_BrainMatter_Jacket_CtrlRig
graphs [
 RigUnit_SequenceAggregate_1_ContainedGraph {nodes [
   n1 = RigVMFunctionEntryNode()
   RigUnit_SequenceAggregate = &RigVMFunction_Sequence(ExecuteContext=@n1.ExecuteContext)
   RigUnit_SequenceAggregate_1 = &RigVMFunction_Sequence(ExecuteContext=@RigUnit_SequenceAggregate.B)
   RigVMFunctionReturnNode(A=@RigUnit_SequenceAggregate.A, B=@RigUnit_SequenceAggregate_1.A, C=@RigUnit_SequenceAggregate_1.B)
  ]}
 RigVMFunctionLibrary {links [], nodes []}
 RigVMModel {nodes [
   RigUnit_BeginExecution = &RigUnit_BeginExecution()
   RigUnit_SequenceAggregate = RigVMAggregateNode(ExecuteContext=@RigUnit_BeginExecution.ExecutePin)
   n1 = &RigUnit_SphericalPoseReader(
    ActiveRegionScaleFactors=(PositiveWidth=1,NegativeWidth=1,PositiveHeight=1,NegativeHeight=1)
    ActiveRegionSize=0.1
    Debug=(bDrawDebug=false,bDraw2D=false,bDrawLocalAxes=false,DebugScale=30,DebugSegments=30,DebugThickness=0.1)
    DriverAxis=(X=-0.5,Y=0,Z=0)
    DriverItem=(Type=Bone,Name="upperarm_r")
    FalloffRegionScaleFactors=(PositiveWidth=1,NegativeWidth=1,PositiveHeight=2,NegativeHeight=1)
    FalloffSize=0.3
    OptionalParentItem=(Type=Bone,Name="None")
    RotationOffset=(X=-120,Y=40,Z=0)
    ExecutePin=@RigUnit_SequenceAggregate.A
   )
   > &RigUnit_ModifyTransforms(
    ItemToModify="((Item=(Type=Bone,Name=\"dyn_collar_r_01\"),Transform=(Rotation=(X=0,Y=0,Z=0,W=1),Translation=(X=0,Y=0,Z=0),Scale3D=(X=1,Y=1,Z=1))),(Item=(Type=Bone,Name=\"Collar_Corrective_R\"),Transform=(Rotation=(X=0,Y=0,Z=0,W=1),Translation=(X=0,Y=0,Z=0),Scale3D=(X=1,Y=1, ...(truncated)"
    ItemToModify.0.Transform=/Script/RigVM.RigVMFunction_MathTransformFromSRT(Location=(X=0,Y=-3,Z=-4.5), Rotation=(X=-30,Y=10,Z=0), Scale=(X=1,Y=1,Z=1)) {is=struct RigVMUnitNode}.Transform
    ItemToModify.1.Transform=&RigVMFunction_MathTransformFromSRT(Location=(X=1,Y=0,Z=0), Rotation=(X=0,Y=0,Z=0), Scale=(X=1,Y=1,Z=1)).Transform
    Weight=@n1.OutputParam
   )
   > n2 = &RigUnit_SphericalPoseReader(
    ActiveRegionScaleFactors=(PositiveWidth=1,NegativeWidth=1,PositiveHeight=1,NegativeHeight=1)
    ActiveRegionSize=0.1
    Debug=(bDrawDebug=false,bDraw2D=false,bDrawLocalAxes=false,DebugScale=30,DebugSegments=30,DebugThickness=0.1)
    DriverAxis=(X=0.5,Y=0,Z=0)
    DriverItem=(Type=Bone,Name="upperarm_l")
    FalloffRegionScaleFactors=(PositiveWidth=1,NegativeWidth=1,PositiveHeight=2,NegativeHeight=1)
    FalloffSize=0.3
    OptionalParentItem=(Type=Bone,Name="None")
    RotationOffset=(X=50,Y=-60,Z=0)
   )
   > &RigUnit_ModifyTransforms(
    ItemToModify="((Item=(Type=Bone,Name=\"dyn_collar_l_base\"),Transform=(Rotation=(X=0,Y=0,Z=0,W=1),Translation=(X=0,Y=0,Z=0),Scale3D=(X=1,Y=1,Z=1))),(Item=(Type=Bone,Name=\"Collar_Corrective_L\"),Transform=(Rotation=(X=0,Y=0,Z=0,W=1),Translation=(X=0,Y=0,Z=0),Scale3D=(X=1,Y= ...(truncated)"
    ItemToModify.0.Transform=&RigVMFunction_MathTransformFromSRT(Location=(X=1,Y=0,Z=2), Rotation=(X=0,Y=0,Z=30), Scale=(X=1,Y=1,Z=1)).Transform
    ItemToModify.1.Transform=&RigVMFunction_MathTransformFromSRT(Location=(X=4,Y=0,Z=0), Rotation=(X=0,Y=0,Z=0), Scale=(X=1,Y=1,Z=1)).Transform
    ItemToModify.2.Transform=&RigVMFunction_MathTransformFromSRT(Location=(X=-1,Y=0,Z=0), Rotation=(X=0,Y=0,Z=-5), Scale=(X=1,Y=1,Z=1)).Transform
    Weight=@n2.OutputParam
   )
   /Script/ControlRig.RigUnit_PositionConstraintLocalSpaceOffset(
    Child=(Type=Bone,Name="dyn_back_fringe_12_01")
    Filter=(bX=false,bY=false,bZ=True)
    Parents=((Item=(Type=Bone,Name="dyn_back_fringe_13_01"),Weight=1))
    Weight=0.2
    bMaintainOffset=True
    ExecutePin=@RigUnit_SequenceAggregate.B
   ) {is=struct RigVMUnitNode}
   > &RigUnit_PositionConstraintLocalSpaceOffset(
    Child=(Type=Bone,Name="dyn_back_fringe_01_01")
    Filter=(bX=false,bY=false,bZ=True)
    Parents=((Item=(Type=Bone,Name="dyn_back_fringe_02_01"),Weight=1))
    Weight=0.2
    bMaintainOffset=True
   )
   RigUnit_GetTransform = &RigUnit_GetTransform(Item=(Type=Bone,Name="dyn_back_fringe_12_01"))
   > RigUnit_SetRotation = &RigUnit_SetRotation(
    Item=(Type=Bone,Name="dyn_back_fringe_13_01")
    Space=GlobalSpace
    Weight=1
    bInitial=False
    Value=/Script/ControlRig.RigUnit_SpringInterpQuaternionV2(CriticalDamping=1, Current=(X=0,Y=0,Z=0,W=1), Strength=7, Torque=(X=0,Y=0,Z=0), Target=@RigUnit_GetTransform.Transform.Rotation) {
     is=struct RigVMUnitNode
     template="SpringInterp::Execute(in Target,in Strength,in CriticalDamping,in Torque,in bUseCurrentInput,in Current,in TargetVelocityAmount,in bInitializeFromTarget,out Result,out AngularVelocity)"
    }.Result
   )
   > &RigUnit_SetRotation(
    Item=(Type=Bone,Name="dyn_back_fringe_13_02")
    Space=GlobalSpace
    Weight=1
    bInitial=False
    Value=&RigUnit_SpringInterpQuaternionV2(
     CriticalDamping=1
     Current=(X=0,Y=0,Z=0,W=1)
     Strength=7
     Torque=(X=0,Y=0,Z=0)
     Target=&RigUnit_GetTransform(Item=(Type=Bone,Name="dyn_back_fringe_12_02")).Transform.Rotation
    ) {template="SpringInterp::Execute(in Target,in Strength,in CriticalDamping,in Torque,in bUseCurrentInput,in Current,in TargetVelocityAmount,in bInitializeFromTarget,out Result,out AngularVelocity)"}.Result
   )
   n3 = &RigUnit_SpringInterpQuaternionV2(
    CriticalDamping=1
    Current=(X=0,Y=0,Z=0,W=1)
    Strength=7
    Torque=(X=0,Y=0,Z=0)
    Target=&RigUnit_GetTransform(Item=(Type=Bone,Name="dyn_back_fringe_10_02")).Transform.Rotation
   ) {template="SpringInterp::Execute(in Target,in Strength,in CriticalDamping,in Torque,in bUseCurrentInput,in Current,in TargetVelocityAmount,in bInitializeFromTarget,out Result,out AngularVelocity)"}
   > &RigUnit_SetRotation(Item=(Type=Bone,Name="dyn_back_fringe_09_02"), Space=GlobalSpace, Weight=1, bInitial=False, Value=@n3.Result)
   > &RigUnit_SetRotation(Item=(Type=Bone,Name="dyn_back_fringe_11_02"), Space=GlobalSpace, Weight=1, bInitial=False, Value=@n3.Result)
   n4 = &RigUnit_SpringInterpQuaternionV2(
    CriticalDamping=1
    Current=(X=0,Y=0,Z=0,W=1)
    Strength=7
    Torque=(X=0,Y=0,Z=0)
    Target=&RigUnit_GetTransform(Item=(Type=Bone,Name="dyn_back_fringe_10_01")).Transform.Rotation
   ) {template="SpringInterp::Execute(in Target,in Strength,in CriticalDamping,in Torque,in bUseCurrentInput,in Current,in TargetVelocityAmount,in bInitializeFromTarget,out Result,out AngularVelocity)"}
   > &RigUnit_SetRotation(Item=(Type=Bone,Name="dyn_back_fringe_09_01"), Space=GlobalSpace, Weight=1, bInitial=False, Value=@n4.Result)
   > &RigUnit_SetRotation(Item=(Type=Bone,Name="dyn_back_fringe_11_01"), Space=GlobalSpace, Weight=1, bInitial=False, Value=@n4.Result)
   n5 = &RigUnit_SpringInterpQuaternionV2(
    CriticalDamping=1
    Current=(X=0,Y=0,Z=0,W=1)
    Strength=7
    Torque=(X=0,Y=0,Z=0)
    Target=&RigUnit_GetTransform(Item=(Type=Bone,Name="dyn_back_fringe_07_01")).Transform.Rotation
   ) {template="SpringInterp::Execute(in Target,in Strength,in CriticalDamping,in Torque,in bUseCurrentInput,in Current,in TargetVelocityAmount,in bInitializeFromTarget,out Result,out AngularVelocity)"}
   > &RigUnit_SetRotation(Item=(Type=Bone,Name="dyn_back_fringe_06_01"), Space=GlobalSpace, Weight=1, bInitial=False, Value=@n5.Result)
   > &RigUnit_SetRotation(Item=(Type=Bone,Name="dyn_back_fringe_08_01"), Space=GlobalSpace, Weight=1, bInitial=False, Value=@n5.Result)
   n6 = &RigUnit_SpringInterpQuaternionV2(
    CriticalDamping=1
    Current=(X=0,Y=0,Z=0,W=1)
    Strength=7
    Torque=(X=0,Y=0,Z=0)
    Target=&RigUnit_GetTransform(Item=(Type=Bone,Name="dyn_back_fringe_07_02")).Transform.Rotation
   ) {template="SpringInterp::Execute(in Target,in Strength,in CriticalDamping,in Torque,in bUseCurrentInput,in Current,in TargetVelocityAmount,in bInitializeFromTarget,out Result,out AngularVelocity)"}
   > &RigUnit_SetRotation(Item=(Type=Bone,Name="dyn_back_fringe_06_02"), Space=GlobalSpace, Weight=1, bInitial=False, Value=@n6.Result)
   > &RigUnit_SetRotation(Item=(Type=Bone,Name="dyn_back_fringe_08_02"), Space=GlobalSpace, Weight=1, bInitial=False, Value=@n6.Result)
   n7 = &RigUnit_SpringInterpQuaternionV2(
    CriticalDamping=1
    Current=(X=0,Y=0,Z=0,W=1)
    Strength=7
    Torque=(X=0,Y=0,Z=0)
    Target=&RigUnit_GetTransform(Item=(Type=Bone,Name="dyn_back_fringe_04_01")).Transform.Rotation
   ) {template="SpringInterp::Execute(in Target,in Strength,in CriticalDamping,in Torque,in bUseCurrentInput,in Current,in TargetVelocityAmount,in bInitializeFromTarget,out Result,out AngularVelocity)"}
   > &RigUnit_SetRotation(Item=(Type=Bone,Name="dyn_back_fringe_05_01"), Space=GlobalSpace, Weight=1, bInitial=False, Value=@n7.Result)
   > &RigUnit_SetRotation(Item=(Type=Bone,Name="dyn_back_fringe_03_01"), Space=GlobalSpace, Weight=1, bInitial=False, Value=@n7.Result)
   n8 = &RigUnit_SpringInterpQuaternionV2(
    CriticalDamping=1
    Current=(X=0,Y=0,Z=0,W=1)
    Strength=7
    Torque=(X=0,Y=0,Z=0)
    Target=&RigUnit_GetTransform(Item=(Type=Bone,Name="dyn_back_fringe_04_02")).Transform.Rotation
   ) {template="SpringInterp::Execute(in Target,in Strength,in CriticalDamping,in Torque,in bUseCurrentInput,in Current,in TargetVelocityAmount,in bInitializeFromTarget,out Result,out AngularVelocity)"}
   > &RigUnit_SetRotation(Item=(Type=Bone,Name="dyn_back_fringe_05_02"), Space=GlobalSpace, Weight=1, bInitial=False, Value=@n8.Result)
   > &RigUnit_SetRotation(Item=(Type=Bone,Name="dyn_back_fringe_03_02"), Space=GlobalSpace, Weight=1, bInitial=False, Value=@n8.Result)
   > &RigUnit_SetRotation(
    Item=(Type=Bone,Name="dyn_back_fringe_01_01")
    Space=GlobalSpace
    Weight=1
    bInitial=False
    Value=&RigUnit_SpringInterpQuaternionV2(
     CriticalDamping=1
     Current=(X=0,Y=0,Z=0,W=1)
     Strength=7
     Torque=(X=0,Y=0,Z=0)
     Target=&RigUnit_GetTransform(Item=(Type=Bone,Name="dyn_back_fringe_02_01")).Transform.Rotation
    ) {template="SpringInterp::Execute(in Target,in Strength,in CriticalDamping,in Torque,in bUseCurrentInput,in Current,in TargetVelocityAmount,in bInitializeFromTarget,out Result,out AngularVelocity)"}.Result
   )
   > &RigUnit_SetRotation(
    Item=(Type=Bone,Name="dyn_back_fringe_01_02")
    Space=GlobalSpace
    Weight=1
    bInitial=False
    Value=&RigUnit_SpringInterpQuaternionV2(
     CriticalDamping=1
     Current=(X=0,Y=0,Z=0,W=1)
     Strength=7
     Torque=(X=0,Y=0,Z=0)
     Target=&RigUnit_GetTransform(Item=(Type=Bone,Name="dyn_back_fringe_02_02")).Transform.Rotation
    ) {template="SpringInterp::Execute(in Target,in Strength,in CriticalDamping,in Torque,in bUseCurrentInput,in Current,in TargetVelocityAmount,in bInitializeFromTarget,out Result,out AngularVelocity)"}.Result
   )
   n9 = &RigUnit_SphericalPoseReader(
    ActiveRegionScaleFactors=(PositiveWidth=1,NegativeWidth=0.2,PositiveHeight=1,NegativeHeight=1)
    ActiveRegionSize=0.3
    Debug=(bDrawDebug=false,bDraw2D=false,bDrawLocalAxes=false,DebugScale=30,DebugSegments=30,DebugThickness=0.2)
    DriverAxis=(X=-1,Y=0,Z=0)
    DriverItem=(Type=Bone,Name="thigh_l")
    FalloffRegionScaleFactors=(PositiveWidth=1,NegativeWidth=1.4,PositiveHeight=1.4,NegativeHeight=1)
    FalloffSize=0.5
    OptionalParentItem=(Type=Bone,Name="pelvis")
    RotationOffset=(X=90,Y=0,Z=-15)
    ExecutePin=@RigUnit_SequenceAggregate.C
   )
   RigVMFunction_MathQuaternionFromEuler = &RigVMFunction_MathQuaternionFromEuler(Euler=(X=0,Y=0,Z=30))
   > &RigUnit_ModifyTransforms(
    ItemToModify=((Item=(Type=Bone,Name="dyn_jacket_f_l_base"),Transform=(Rotation=(X=0,Y=0,Z=0,W=1),Translation=(X=0,Y=0,Z=0),Scale3D=(X=1,Y=1,Z=1))))
    ItemToModify.0.Transform.Rotation=@RigVMFunction_MathQuaternionFromEuler.Result
    Weight=@n9.OutputParam
   )
   > n10 = &RigUnit_SphericalPoseReader(
    ActiveRegionScaleFactors=(PositiveWidth=1,NegativeWidth=0,PositiveHeight=1,NegativeHeight=1)
    ActiveRegionSize=0.3
    Debug=(bDrawDebug=false,bDraw2D=false,bDrawLocalAxes=false,DebugScale=30,DebugSegments=30,DebugThickness=0.2)
    DriverAxis=(X=1,Y=0,Z=0)
    DriverItem=(Type=Bone,Name="thigh_r")
    FalloffRegionScaleFactors=(PositiveWidth=1,NegativeWidth=1.4,PositiveHeight=1,NegativeHeight=1)
    FalloffSize=0.5
    OptionalParentItem=(Type=Bone,Name="pelvis")
    RotationOffset=(X=-90,Y=180,Z=-14)
   )
   > /Script/ControlRig.RigUnit_SetRelativeTransformForItem(
    Child=(Type=Bone,Name="dyn_jacket_f_r_base")
    Parent=(Type=Bone,Name="dyn_jacket_f_r_base")
    Value=(Rotation=(X=0,Y=0,Z=0,W=1),Translation=(X=0,Y=0,Z=0),Scale3D=(X=1,Y=1,Z=1))
    bParentInitial=False
    bPropagateToChildren=True
    Value.Rotation=&RigVMFunction_MathQuaternionFromEuler(Euler=(X=0,Y=0,Z=-15)).Result
    Weight=@n10.OutputParam
   ) {
    is=struct RigVMUnitNode
    template="Set Relative Transform::Execute(in Child,in Parent,in bParentInitial,in Value,in Weight,in bPropagateToChildren)"
   }
   > n11 = &RigUnit_SphericalPoseReader(
    ActiveRegionScaleFactors=(PositiveWidth=2,NegativeWidth=2,PositiveHeight=2,NegativeHeight=1)
    ActiveRegionSize=0.05
    Debug=(bDrawDebug=false,bDraw2D=false,bDrawLocalAxes=false,DebugScale=30,DebugSegments=30,DebugThickness=0.2)
    DriverAxis=(X=1,Y=0,Z=0)
    DriverItem=(Type=Bone,Name="spine_04")
    FalloffRegionScaleFactors=(PositiveWidth=1,NegativeWidth=1,PositiveHeight=1,NegativeHeight=1)
    FalloffSize=0.1
    OptionalParentItem=(Type=Bone,Name="pelvis")
    RotationOffset=(X=0,Y=-90,Z=45)
   )
   Multiply = &RigVMFunction_MathFloatMul(A=@n11.OutputParam, B=@n9.OutputParam) {template="Multiply::Execute(in A,in B,out Result)"}
   > RigUnit_SetRelativeTransformForItem = &RigUnit_SetRelativeTransformForItem(
    Child=(Type=Bone,Name="dyn_jacket_f_l_base")
    Parent=(Type=Bone,Name="dyn_jacket_f_l_base")
    Value=(Rotation=(X=0,Y=0,Z=0,W=1),Translation=(X=-5,Y=2,Z=0),Scale3D=(X=1,Y=1,Z=1))
    bParentInitial=False
    bPropagateToChildren=True
    Weight=@Multiply.Result
    Value.Rotation=&RigVMFunction_MathQuaternionFromEuler(Euler=(X=0,Y=0,Z=-4)).Result
   ) {template="Set Relative Transform::Execute(in Child,in Parent,in bParentInitial,in Value,in Weight,in bPropagateToChildren)"}
   RigVMFunction_MathQuaternionToEuler = &RigVMFunction_MathQuaternionToEuler(
    RotationOrder=ZYX
    Value=&RigUnit_GetRelativeTransformForItem(Child=(Type=Bone,Name="spine_05"), Parent=(Type=Bone,Name="spine_01")).RelativeTransform.Rotation
   )
   > &RigUnit_SetRelativeTransformForItem(
    Child=(Type=Bone,Name="dyn_jacket_f_r_base")
    Parent=(Type=Bone,Name="dyn_jacket_f_r_base")
    Value=(Rotation=(X=0,Y=0,Z=0,W=1),Translation=(X=-5,Y=0,Z=-3),Scale3D=(X=1,Y=1,Z=1))
    bParentInitial=False
    bPropagateToChildren=True
    Weight=&RigVMFunction_MathDoubleRemap(SourceMaximum=60, SourceMinimum=20, TargetMaximum=1, TargetMinimum=0, Value=@RigVMFunction_MathQuaternionToEuler.Result.X).Result
    Value.Rotation=&RigVMFunction_MathQuaternionFromEuler(Euler=(X=-30,Y=-30,Z=20)).Result
   ) {template="Set Relative Transform::Execute(in Child,in Parent,in bParentInitial,in Value,in Weight,in bPropagateToChildren)"}
   Remap = &RigVMFunction_MathDoubleRemap(SourceMaximum=-70, SourceMinimum=-20, TargetMaximum=1, TargetMinimum=0, Value=@RigVMFunction_MathQuaternionToEuler.Result.X)
   > &RigUnit_SetRelativeTransformForItem(
    Child=(Type=Bone,Name="dyn_jacket_f_l_base")
    Parent=(Type=Bone,Name="dyn_jacket_f_l_base")
    Value=(Rotation=(X=0,Y=0,Z=0,W=1),Translation=(X=-10,Y=2,Z=3),Scale3D=(X=1,Y=1,Z=1))
    bParentInitial=False
    bPropagateToChildren=True
    Weight=@Remap.Result
    Value.Rotation=&RigVMFunction_MathQuaternionFromEuler(Euler=(X=-30,Y=-40,Z=-40)).Result
   ) {template="Set Relative Transform::Execute(in Child,in Parent,in bParentInitial,in Value,in Weight,in bPropagateToChildren)"}
   n12 = &RigVMFunction_MathDoubleRemap(SourceMaximum=70, SourceMinimum=20, TargetMaximum=1, TargetMinimum=0, Value=@RigVMFunction_MathQuaternionToEuler.Result.Z)
   > &RigUnit_SetRelativeTransformForItem(
    Child=(Type=Bone,Name="dyn_jacket_f_l_base")
    Parent=(Type=Bone,Name="dyn_jacket_f_l_base")
    Value=(Rotation=(X=0,Y=0,Z=0,W=1),Translation=(X=-10,Y=-2,Z=10),Scale3D=(X=1,Y=1,Z=1))
    bParentInitial=False
    bPropagateToChildren=True
    Value.Rotation=&RigVMFunction_MathQuaternionFromEuler(Euler=(X=0,Y=-50,Z=-40)).Result
    Weight=@n12.Result
   ) {template="Set Relative Transform::Execute(in Child,in Parent,in bParentInitial,in Value,in Weight,in bPropagateToChildren)"}
   > &RigUnit_SetRelativeTransformForItem(
    Child=(Type=Bone,Name="dyn_jacket_f_r_base")
    Parent=(Type=Bone,Name="dyn_jacket_f_r_base")
    Value=(Rotation=(X=0,Y=0,Z=0,W=1),Translation=(X=-8,Y=0,Z=0),Scale3D=(X=1,Y=1,Z=1))
    bParentInitial=False
    bPropagateToChildren=True
    Value.Rotation=&RigVMFunction_MathQuaternionFromEuler(Euler=(X=0,Y=-60,Z=60)).Result
    Weight=@n12.Result
   ) {template="Set Relative Transform::Execute(in Child,in Parent,in bParentInitial,in Value,in Weight,in bPropagateToChildren)"}
   > &RigUnit_SetRelativeTransformForItem(
    Child=(Type=Bone,Name="dyn_jacket_f_l_base")
    Parent=(Type=Bone,Name="dyn_jacket_f_l_base")
    Value=(Rotation=(X=0,Y=0,Z=0,W=1),Translation=(X=0,Y=0,Z=0),Scale3D=(X=1,Y=1,Z=1))
    bParentInitial=False
    bPropagateToChildren=True
    Value.Rotation=&RigVMFunction_MathQuaternionFromEuler(Euler=(X=0,Y=0,Z=-30)).Result
    Weight=&RigVMFunction_MathDoubleMul(
     A=&RigVMFunction_MathDoubleRemap(SourceMaximum=20, SourceMinimum=6, TargetMaximum=1, TargetMinimum=0, Value=@RigVMFunction_MathQuaternionToEuler.Result.Z).Result
     B=@n9.OutputParam
    ).Result
   ) {template="Set Relative Transform::Execute(in Child,in Parent,in bParentInitial,in Value,in Weight,in bPropagateToChildren)"}
  ]}
]
hierarchy [
 root {loc=<0 0 0>, rot=<0 0 0 1>}
 pelvis {parent=root, loc=<0 -0.5301 98.0571>, rot=<0.0127 0.707 -0.0127 -0.707>}
 spine_01 {parent=pelvis, loc=<6.1208 -0.9328 0>, rot=<0 0 0.034 -0.9994>}
 spine_02 {parent=spine_01, loc=<6.5996 -0.2095 0>, rot=<0 0 0.0238 -0.9997>}
 spine_03 {parent=spine_02, loc=<6.4914 -0.1237 0>, rot=<0 0 0.0149 -0.9999>}
 spine_04 {parent=spine_03, loc=<6.9598 0.0794 0>, rot=<0 0 -0.0132 -0.9999>}
 spine_05 {parent=spine_04, loc=<6.9987 0 0>, rot=<0 0 -0.0015 -1>}
 attach_cape {parent=spine_05, loc=<9.3216 -8.4501 0>, rot=<0.0616 -0.7044 -0.0616 -0.7044>}
 neck_01 {parent=spine_05, loc=<11.3138 0.1529 0>, rot=<0 0 -0.2003 -0.9797>}
 neck_02 {parent=neck_01, loc=<4.5627 0 0>, rot=<0 0 0.0504 -0.9987>}
 head {parent=neck_02, loc=<5.3914 0 0>, rot=<0 0 0.115 -0.9934>}
 attach_backpack {parent=spine_05, loc=<0 -8.35 0>, rot=<-0.0093 -0.707 0.0093 -0.707>}
 clavicle_l {parent=spine_05, loc=<5.4831 -0.1053 -3.2946>, rot=<0.0465 -0.7289 -0.009 -0.6829>}
 upperarm_l {parent=clavicle_l, loc=<9.8658 -1.2136 -0.4105>, rot=<0.0164 -0.3245 -0.0269 -0.9454>}
 lowerarm_l {parent=upperarm_l, loc=<26.465 -0.8345 0.259>, rot=<0.0036 0.0161 -0.2298 -0.9731>}
 hand_l {parent=lowerarm_l, loc=<22.7355 -0.3203 -0.6105>, rot=<-0.6784 -0.0223 0.0735 -0.7307>}
 thumb_01_l {parent=hand_l, loc=<1.3118 -0.3191 -2.459>, rot=<0.6546 -0.4042 0.0664 -0.6354>}
 thumb_02_l {parent=thumb_01_l, loc=<5.9336 -0.0448 0.0401>, rot=<-0.0037 0.0213 0.298 -0.9543>}
 thumb_03_l {parent=thumb_02_l, loc=<2.2223 0.3682 -0.0907>, rot=<0.0118 -0.013 0.0066 -0.9998>}
 index_metacarpal_l {parent=hand_l, loc=<1.6535 0.0183 -2.96>, rot=<0.0158 -0.0124 -0.0173 -0.9996>}
 index_01_l {parent=index_metacarpal_l, loc=<8.3008 0.0357 -0.0524>, rot=<-0.0044 0.0485 0.0924 -0.9945>}
 index_02_l {parent=index_01_l, loc=<3.3587 -0.3004 0.0452>, rot=<-0.002 0.0027 0.3225 -0.9466>}
 index_03_l {parent=index_02_l, loc=<2.5578 -0.123 -0.0252>, rot=<-0.002 0.0067 0.2558 -0.9667>}
 middle_metacarpal_l {parent=hand_l, loc=<2.1856 -0.1725 -0.5984>, rot=<0.006 0.0168 -0.0329 -0.9993>}
 middle_01_l {parent=middle_metacarpal_l, loc=<7.3275 -0.2447 0.0129>, rot=<-0.0208 0.0042 0.0858 -0.9961>}
 middle_02_l {parent=middle_01_l, loc=<4.0123 0.1401 0.1175>, rot=<-0.0218 0.0258 0.3512 -0.9357>}
 middle_03_l {parent=middle_02_l, loc=<2.6077 0.0534 0.1373>, rot=<0.0082 0.0188 0.1708 -0.9851>}
 ring_metacarpal_l {parent=hand_l, loc=<2.2074 -0.3625 1.3638>, rot=<-0.0284 0.0722 -0.0458 -0.9959>}
 ring_01_l {parent=ring_metacarpal_l, loc=<7.0295 -0.6197 -0.1446>, rot=<-0.0489 -0.0252 0.1892 -0.9804>}
 ring_02_l {parent=ring_01_l, loc=<3.6174 0.4647 -0.1027>, rot=<0 0 0.2806 -0.9598>}
 ring_03_l {parent=ring_02_l, loc=<2.4152 0.114 0.0199>, rot=<0 0 0.1189 -0.9929>}
 pinky_metacarpal_l {parent=hand_l, loc=<1.7616 -0.9753 3.3814>, rot=<-0.0397 0.0814 -0.069 -0.9935>}
 pinky_01_l {parent=pinky_metacarpal_l, loc=<7.1709 -0.4831 -0.1094>, rot=<-0.0654 -0.0035 0.2242 -0.9723>}
 pinky_02_l {parent=pinky_01_l, loc=<2.8274 0.0893 -0.1043>, rot=<0.0037 0.0011 0.2854 -0.9584>}
 pinky_03_l {parent=pinky_02_l, loc=<2.2046 0.0307 -0.0078>, rot=<-0.0002 0.0098 0.1445 -0.9895>}
 deform_wrist_dn_base_l {parent=hand_l, loc=<0 -1.5326 0>, rot=<0 0 0 -1>}
 deform_wrist_dn_l {parent=deform_wrist_dn_base_l, loc=<0 0 0>, rot=<0 0 0 1>}
 deform_wrist_up_base_l {parent=hand_l, loc=<0 2.3009 0>, rot=<0 0 0 -1>}
 deform_wrist_up_l {parent=deform_wrist_up_base_l, loc=<0 0 0>, rot=<0 0 0 1>}
 lowerarm_twist_02_l {parent=lowerarm_l, loc=<10.6707 0 0>, rot=<0 0 0 1>}
 dyn_arm_fringe_l_03_base {parent=lowerarm_twist_02_l, loc=<1.2804 -0.3473 -3.4744>, rot=<0.5951 -0.3432 -0.2502 -0.6822>}
 dyn_arm_fringe_l_03_02 {parent=dyn_arm_fringe_l_03_base, loc=<0 0 0>, rot=<0 0 0 1>}
 dyn_arm_fringe_l_03_01 {parent=dyn_arm_fringe_l_03_02, loc=<9.3938 0.2894 -0.1454>, rot=<0 0 0 1>}
 lowerarm_twist_01_l {parent=lowerarm_l, loc=<18.316 0 0>, rot=<0 0 0 1>}
 deform_elbow_in_base_l {parent=lowerarm_l, loc=<0.1834 3.6062 -0.2572>, rot=<0 0 0 -1>}
 deform_elbow_in_l {parent=deform_elbow_in_base_l, loc=<0 0 0>, rot=<0 0 0 1>}
 deform_elbow_out_base_l {parent=lowerarm_l, loc=<0 -4 0>, rot=<0 0 0 1>}
 deform_elbow_out_l {parent=deform_elbow_out_base_l, loc=<0 0 0>, rot=<0 0 0 1>}
 dyn_arm_fringe_l_01_base {parent=lowerarm_l, loc=<1.0137 2.0331 -4.7309>, rot=<0.5216 -0.3944 -0.2129 -0.726>}
 dyn_arm_fringe_l_01_01 {parent=dyn_arm_fringe_l_01_base, loc=<0 0 0>, rot=<0 0 0 1>}
 dyn_arm_fringe_l_01_02 {parent=dyn_arm_fringe_l_01_01, loc=<8.187 -0.0643 0.0788>, rot=<0 0 0 1>}
 dyn_arm_fringe_l_02_base {parent=lowerarm_l, loc=<6.6233 0.0931 -4.2238>, rot=<0.5494 -0.3721 -0.2571 -0.7026>}
 dyn_arm_fringe_l_02_01 {parent=dyn_arm_fringe_l_02_base, loc=<0 0 0>, rot=<0 0 0 1>}
 dyn_arm_fringe_l_02_02 {parent=dyn_arm_fringe_l_02_01, loc=<9.6374 0.0067 -0.2126>, rot=<0 0 0 1>}
 upperarm_twist_01_l {parent=upperarm_l, loc=<7.2135 -0.1134 0>, rot=<0 0 0 1>}
 deform_upperarm_bk_base_l {parent=upperarm_twist_01_l, loc=<-6.4053 -4.3746 0.314>, rot=<0 0 0 -1>}
 deform_upperarm_bk_l {parent=deform_upperarm_bk_base_l, loc=<0 0 0>, rot=<0 0 0 1>}
 deform_upperarm_fr_base_l {parent=upperarm_twist_01_l, loc=<-5.6312 5.092 0.3354>, rot=<0 0 0 -1>}
 deform_upperarm_fr_l {parent=deform_upperarm_fr_base_l, loc=<0 0 0>, rot=<0 0 0 1>}
 deform_upperarm_up_base_l {parent=upperarm_twist_01_l, loc=<-8.891 0.4825 3.9517>, rot=<0 0 0 -1>}
 deform_upperarm_up_l {parent=deform_upperarm_up_base_l, loc=<0 0 0>, rot=<0 0 0 1>}
 upperarm_twist_02_l {parent=upperarm_l, loc=<19.7998 -0.1134 0>, rot=<0 0 0 1>}
 clavicle_r {parent=spine_05, loc=<5.4831 -0.1053 3.2946>, rot=<0.7289 0.0465 -0.6829 0.009>}
 upperarm_r {parent=clavicle_r, loc=<-9.8658 1.2136 0.4105>, rot=<0.0164 -0.3245 -0.0269 -0.9454>}
 lowerarm_r {parent=upperarm_r, loc=<-26.465 0.8345 -0.259>, rot=<0.0036 0.0161 -0.2298 -0.9731>}
 hand_r {parent=lowerarm_r, loc=<-22.7355 0.3203 0.6106>, rot=<-0.6784 -0.0223 0.0735 -0.7307>}
 thumb_01_r {parent=hand_r, loc=<-1.3118 0.3191 2.459>, rot=<0.6546 -0.4042 0.0664 -0.6354>}
 thumb_02_r {parent=thumb_01_r, loc=<-5.9336 0.0448 -0.0401>, rot=<-0.0037 0.0213 0.298 -0.9543>}
 thumb_03_r {parent=thumb_02_r, loc=<-2.2223 -0.3682 0.0907>, rot=<0.0118 -0.013 0.0066 -0.9998>}
 index_metacarpal_r {parent=hand_r, loc=<-1.6535 -0.0183 2.96>, rot=<0.0158 -0.0124 -0.0173 -0.9996>}
 index_01_r {parent=index_metacarpal_r, loc=<-8.3008 -0.0357 0.0524>, rot=<-0.0044 0.0485 0.0924 -0.9945>}
 index_02_r {parent=index_01_r, loc=<-3.3587 0.3004 -0.0452>, rot=<-0.002 0.0027 0.3225 -0.9466>}
 index_03_r {parent=index_02_r, loc=<-2.5578 0.123 0.0252>, rot=<-0.002 0.0067 0.2558 -0.9667>}
 middle_metacarpal_r {parent=hand_r, loc=<-2.1856 0.1725 0.5984>, rot=<0.006 0.0168 -0.0329 -0.9993>}
 middle_01_r {parent=middle_metacarpal_r, loc=<-7.3275 0.2447 -0.0129>, rot=<-0.0208 0.0042 0.0858 -0.9961>}
 middle_02_r {parent=middle_01_r, loc=<-4.0123 -0.1401 -0.1175>, rot=<-0.0218 0.0258 0.3512 -0.9357>}
 middle_03_r {parent=middle_02_r, loc=<-2.6077 -0.0534 -0.1373>, rot=<0.0082 0.0188 0.1708 -0.9851>}
 ring_metacarpal_r {parent=hand_r, loc=<-2.2074 0.3625 -1.3638>, rot=<-0.0284 0.0722 -0.0458 -0.9959>}
 ring_01_r {parent=ring_metacarpal_r, loc=<-7.0295 0.6197 0.1446>, rot=<-0.0489 -0.0252 0.1892 -0.9804>}
 ring_02_r {parent=ring_01_r, loc=<-3.6174 -0.4647 0.1027>, rot=<0 0 0.2806 -0.9598>}
 ring_03_r {parent=ring_02_r, loc=<-2.4152 -0.114 -0.0199>, rot=<0 0 0.1189 -0.9929>}
 pinky_metacarpal_r {parent=hand_r, loc=<-1.7616 0.9753 -3.3814>, rot=<-0.0397 0.0814 -0.069 -0.9935>}
 pinky_01_r {parent=pinky_metacarpal_r, loc=<-7.1709 0.4831 0.1094>, rot=<-0.0654 -0.0035 0.2242 -0.9723>}
 pinky_02_r {parent=pinky_01_r, loc=<-2.8274 -0.0893 0.1043>, rot=<0.0037 0.0011 0.2854 -0.9584>}
 pinky_03_r {parent=pinky_02_r, loc=<-2.2046 -0.0307 0.0078>, rot=<-0.0002 0.0098 0.1445 -0.9895>}
 deform_wrist_dn_base_r {parent=hand_r, loc=<-0.0001 1.5328 0>, rot=<0 0 0 -1>}
 deform_wrist_dn_r {parent=deform_wrist_dn_base_r, loc=<0 0 0>, rot=<0 0 0 1>}
 deform_wrist_up_base_r {parent=hand_r, loc=<-0.0002 -2.3007 -0.0001>, rot=<0 0 0 -1>}
 deform_wrist_up_r {parent=deform_wrist_up_base_r, loc=<0 0 0>, rot=<0 0 0 1>}
 lowerarm_twist_02_r {parent=lowerarm_r, loc=<-10.6707 0 0>, rot=<0 0 0 -1>}
 dyn_arm_fringe_r_03_base {parent=lowerarm_twist_02_r, loc=<-0.4494 0.6663 3.5946>, rot=<0.2749 -0.6595 0.6168 0.3303>}
 dyn_arm_fringe_r_03_01 {parent=dyn_arm_fringe_r_03_base, loc=<0 0 0.0001>, rot=<0 0 0 1>}
 dyn_arm_fringe_r_03_02 {parent=dyn_arm_fringe_r_03_01, loc=<10.665 0.2776 -0.1672>, rot=<0 0 0 1>}
 lowerarm_twist_01_r {parent=lowerarm_r, loc=<-18.316 0 0>, rot=<0 0 0 -1>}
 deform_elbow_in_base_r {parent=lowerarm_r, loc=<-0.1834 -3.6062 0.2572>, rot=<0 0 0 -1>}
 deform_elbow_in_r {parent=deform_elbow_in_base_r, loc=<0 0 0>, rot=<0 0 0 1>}
 deform_elbow_out_base_r {parent=lowerarm_r, loc=<0 4 -0.0001>, rot=<0 0 0 -1>}
 deform_elbow_out_r {parent=deform_elbow_out_base_r, loc=<0 0 0>, rot=<0 0 0 1>}
 dyn_arm_fringe_r_02_base {parent=lowerarm_r, loc=<-6.4375 -0.6509 4.4708>, rot=<0.2566 -0.7171 0.5464 0.3484>}
 dyn_arm_fringe_r_02_01 {parent=dyn_arm_fringe_r_02_base, loc=<0 0 0>, rot=<0 0 0 1>}
 dyn_arm_fringe_r_02_02 {parent=dyn_arm_fringe_r_02_01, loc=<10.2089 0.1633 -0.7289>, rot=<0 0 0 1>}
 dyn_arm_fringe_r_01_base {parent=lowerarm_r, loc=<-1.4694 -1.8097 5.3351>, rot=<0.2214 -0.7324 0.518 0.3824>}
 dyn_arm_fringe_r_01_01 {parent=dyn_arm_fringe_r_01_base, loc=<0 0 0>, rot=<0 0 0 1>}
 dyn_arm_fringe_r_01_02 {parent=dyn_arm_fringe_r_01_01, loc=<7.6207 0.0085 -0.0776>, rot=<0 0 0 1>}
 upperarm_twist_01_r {parent=upperarm_r, loc=<-7.2135 0.1134 0>, rot=<0 0 0 -1>}
 deform_upperarm_bk_base_r {parent=upperarm_twist_01_r, loc=<6.4053 4.3746 -0.314>, rot=<0 0 0 -1>}
 deform_upperarm_bk_r {parent=deform_upperarm_bk_base_r, loc=<0 0 0>, rot=<0 0 0 1>}
 deform_upperarm_fr_base_r {parent=upperarm_twist_01_r, loc=<5.6312 -5.092 -0.3354>, rot=<0 0 0 -1>}
 deform_upperarm_fr_r {parent=deform_upperarm_fr_base_r, loc=<0 0 0>, rot=<0 0 0 1>}
 deform_upperarm_up_base_r {parent=upperarm_twist_01_r, loc=<8.891 -0.4825 -3.9517>, rot=<0 0 0 -1>}
 deform_upperarm_up_r {parent=deform_upperarm_up_base_r, loc=<0 0 0>, rot=<0 0 0 1>}
 upperarm_twist_02_r {parent=upperarm_r, loc=<-19.7998 0.1134 0>, rot=<0 0 0 -1>}
 dyn_collar_r_base {parent=spine_05, loc=<6.0152 6.0419 8.3606>, rot=<-0.171 -0.0302 -0.171 -0.9698>}
 dyn_collar_r_01 {parent=dyn_collar_r_base, loc=<0 0 0>, rot=<0 0 0 1>}
 dyn_collar_r_02 {parent=dyn_collar_r_01, loc=<3.137 0 0>, rot=<0 -0.9063 0 0.4226>}
 dyn_collar_r_03 {parent=dyn_collar_r_02, loc=<4.298 -0.0002 -0.0002>, rot=<0 0 0 1>}
 dyn_collar_l_base {parent=spine_05, loc=<6.0149 6.0419 -8.3606>, rot=<0.171 0.0302 -0.171 -0.9698>}
 dyn_collar_l_01 {parent=dyn_collar_l_base, loc=<0.0003 0 0>, rot=<0 0 0 1>}
 dyn_collar_l_02 {parent=dyn_collar_l_01, loc=<3.1372 0 0>, rot=<0 -0.9063 0 -0.4226>}
 dyn_collar_l_03 {parent=dyn_collar_l_02, loc=<4.2981 0 0>, rot=<0 0 0 1>}
 dyn_jacket_f_r_base {parent=spine_05, loc=<-13.1699 8.7522 12.4127>, rot=<0.0646 -0.9084 -0.4086 0.0608>}
 dyn_jacket_f_r_01 {parent=dyn_jacket_f_r_base, loc=<-0.0002 0 -0.0001>, rot=<0 0 0 1>}
 dyn_jacket_f_r_02 {parent=dyn_jacket_f_r_01, loc=<6.621 0 0>, rot=<0 0 0 1>}
 dyn_jacket_f_l_base {parent=spine_05, loc=<-13.1697 8.7522 -12.4127>, rot=<-0.0608 0.4086 0.9084 -0.0646>}
 dyn_jacket_f_l_01 {parent=dyn_jacket_f_l_base, loc=<0 0 0>, rot=<0 0 0 1>}
 dyn_jacket_f_l_02 {parent=dyn_jacket_f_l_01, loc=<6.621 0 0>, rot=<0 0 0 1>}
 Collar_Corrective_R {parent=spine_05, loc=<12.637 0 8>, rot=<-0.0597 -0.3082 -0.356 -0.8802>}
 Collar_Corrective_L {parent=spine_05, loc=<12.637 -0.0001 -8>, rot=<0.8739 0.3336 0.3322 0.1209>}
 jacket_corrective_back {parent=spine_05, loc=<-11 -11 0.0095>, rot=<0 0 0.9995 0.0327>}
 dyn_back_fringe_01_Base {parent=spine_05, loc=<-0.3525 -11.4618 -13.4763>, rot=<-0.0301 0.702 0.7109 -0.0304>}
 dyn_back_fringe_01_01 {parent=dyn_back_fringe_01_Base, loc=<3.3888 -0.4854 0.579>, rot=<0 0 0 1>}
 dyn_back_fringe_01_02 {parent=dyn_back_fringe_01_01, loc=<9.4626 0 0>, rot=<0 0 0 1>}
 dyn_back_fringe_02_01 {parent=dyn_back_fringe_01_Base, loc=<3.3888 1.7603 -0.012>, rot=<0 0 -0.0132 -0.9999>}
 dyn_back_fringe_02_02 {parent=dyn_back_fringe_02_01, loc=<9.4626 0 0>, rot=<0 0 0 1>}
 dyn_back_fringe_02_base {parent=spine_05, loc=<0.2757 -12.0635 -7.4612>, rot=<-0.0316 0.675 0.736 -0.04>}
 dyn_back_fringe_03_01 {parent=dyn_back_fringe_02_base, loc=<3.3167 -1.6878 0>, rot=<0 0 0 1>}
 dyn_back_fringe_03_02 {parent=dyn_back_fringe_03_01, loc=<9.4544 0 0>, rot=<0 0 0 1>}
 dyn_back_fringe_05_01 {parent=dyn_back_fringe_02_base, loc=<3.3167 2.8494 0>, rot=<0 0 0 1>}
 dyn_back_fringe_05_02 {parent=dyn_back_fringe_05_01, loc=<9.4544 0 0>, rot=<0 0 0 1>}
 dyn_back_fringe_04_01 {parent=dyn_back_fringe_02_base, loc=<3.3167 0.617 0>, rot=<0 0 0 1>}
 dyn_back_fringe_04_02 {parent=dyn_back_fringe_04_01, loc=<9.4544 0 0>, rot=<0 0 0 1>}
 dyn_back_fringe_03_base {parent=spine_05, loc=<0.5025 -12.212 0.0162>, rot=<0.0605 -0.7173 -0.6917 0.0583>}
 dyn_back_fringe_07_01 {parent=dyn_back_fringe_03_base, loc=<3.5661 0.0005 0.4574>, rot=<0 0 0 1>}
 dyn_back_fringe_07_02 {parent=dyn_back_fringe_07_01, loc=<9.6002 0 0>, rot=<0 0 0 1>}
 dyn_back_fringe_06_01 {parent=dyn_back_fringe_03_base, loc=<3.5661 -2.4851 0.4574>, rot=<0 0 0 1>}
 dyn_back_fringe_06_02 {parent=dyn_back_fringe_06_01, loc=<9.6002 0 0>, rot=<0 0 0 1>}
 dyn_back_fringe_08_01 {parent=dyn_back_fringe_03_base, loc=<3.5661 1.9445 0.4574>, rot=<0 0 0 1>}
 dyn_back_fringe_08_02 {parent=dyn_back_fringe_08_01, loc=<9.6002 0 0>, rot=<0 0 0 1>}
 dyn_back_fringe_04_base {parent=spine_05, loc=<0.2756 -12.0634 7.4612>, rot=<0.0363 -0.7242 -0.6873 0.0424>}
 dyn_back_fringe_11_01 {parent=dyn_back_fringe_04_base, loc=<3.3681 1.1836 0>, rot=<0 0 0 1>}
 dyn_back_fringe_11_02 {parent=dyn_back_fringe_11_01, loc=<9.6349 0 0>, rot=<-0.01 0.0004 -0.0022 -0.9999>}
 dyn_back_fringe_10_01 {parent=dyn_back_fringe_04_base, loc=<3.3681 -1.4348 0>, rot=<0 0 0 1>}
 dyn_back_fringe_10_02 {parent=dyn_back_fringe_10_01, loc=<9.6349 0 0>, rot=<-0.01 0.0004 -0.0022 -0.9999>}
 dyn_back_fringe_09_01 {parent=dyn_back_fringe_04_base, loc=<3.3681 -3.4376 0>, rot=<0 0 0 1>}
 dyn_back_fringe_09_02 {parent=dyn_back_fringe_09_01, loc=<9.6349 0 0>, rot=<-0.01 0.0004 -0.0022 -0.9999>}
 dyn_back_fringe_5_base {parent=spine_05, loc=<-0.3524 -11.4617 13.4763>, rot=<0.0394 -0.7154 -0.6971 0.0254>}
 dyn_back_fringe_13_01 {parent=dyn_back_fringe_5_base, loc=<3.2076 -0.0107 0>, rot=<0 0 0 1>}
 dyn_back_fringe_13_02 {parent=dyn_back_fringe_13_01, loc=<9.6362 0 0>, rot=<0 0 0 1>}
 dyn_back_fringe_12_01 {parent=dyn_back_fringe_5_base, loc=<3.2076 -2.1763 0>, rot=<0 0 0 1>}
 dyn_back_fringe_12_02 {parent=dyn_back_fringe_12_01, loc=<9.6362 0 0>, rot=<0 0 0 1>}
 deform_pec_aim_base_l {parent=spine_04, loc=<3.2456 9.3309 -1.2237>, rot=<0.0603 -0.4033 0.3318 -0.8507>}
 deform_pec_aim_l {parent=deform_pec_aim_base_l, loc=<0 0 0>, rot=<0 0 0 1>}
 deform_pec_base_l {parent=deform_pec_aim_l, loc=<9.3665 0 0>, rot=<0 0 0 1>}
 deform_pec_l {parent=deform_pec_base_l, loc=<0 0 0>, rot=<0 0 0 1>}
 deform_pec_aim_base_r {parent=spine_04, loc=<3.2452 9.3309 1.2237>, rot=<-0.4033 -0.0603 0.8507 0.3318>}
 deform_pec_aim_r {parent=deform_pec_aim_base_r, loc=<0 0 0>, rot=<0 0 0 1>}
 deform_pec_base_r {parent=deform_pec_aim_r, loc=<-9.3667 -0.0003 -0.0004>, rot=<0 0 0 1>}
 deform_pec_r {parent=deform_pec_base_r, loc=<0 0 0>, rot=<0 0 0 1>}
 thigh_l {parent=pelvis, loc=<-1.0855 -0.2109 -8.7053>, rot=<0.0009 0.0355 0.0155 -0.9993>}
 calf_l {parent=thigh_l, loc=<-42.5284 -0.0019 -0.7297>, rot=<0.0006 -0.0161 -0.0391 -0.9991>}
 foot_l {parent=calf_l, loc=<-46.7401 -2.5014 -0.592>, rot=<0.0427 -0.0238 0.0353 -0.9982>}
 ball_l {parent=foot_l, loc=<-3.5658 10.8066 -0.1544>, rot=<0.0001 -0.0001 -0.7063 -0.7079>}
 calf_twist_02_l {parent=calf_l, loc=<-19.1546 -1.7943 -0.2638>, rot=<0.0237 0.0095 -0.0035 -0.9997>}
 dyn_bows_leg_l_base {parent=calf_twist_02_l, loc=<0.7074 5.7231 -0.8623>, rot=<0.6816 0.555 -0.3855 -0.2806>}
 dyn_bows_leg_l_01 {parent=dyn_bows_leg_l_base, loc=<0 -0.2903 0>, rot=<-0.0221 -0.0227 0.716 -0.6974>}
 dyn_bows_leg_l_02 {parent=dyn_bows_leg_l_base, loc=<0 0.3134 0>, rot=<0.0287 0.0454 -0.6425 -0.7644>}
 calf_twist_01_l {parent=calf_l, loc=<-35.8483 -1.9197 -0.5794>, rot=<0.0237 0.0095 -0.0035 -0.9997>}
 deform_knee_in_base_l {parent=calf_l, loc=<0.8031 -5.0231 0.2456>, rot=<0 0 0 -1>}
 deform_knee_in_l {parent=deform_knee_in_base_l, loc=<0 0 0>, rot=<0 0 0 1>}
 deform_knee_out_base_l {parent=calf_l, loc=<0 4 0>, rot=<0 0 0 -1>}
 deform_knee_out_l {parent=deform_knee_out_base_l, loc=<0 0 0>, rot=<0 0 0 1>}
 thigh_twist_01_l {parent=thigh_l, loc=<-27.952 0 0>, rot=<0.0013 0 0 -1>}
 deform_groin_base_l {parent=thigh_l, loc=<0.4801 5.6776 -0.2486>, rot=<0 0 0 -1>}
 deform_groin_l {parent=deform_groin_base_l, loc=<0 0 0>, rot=<0 0 0 1>}
 thigh_r {parent=pelvis, loc=<-1.0855 -0.2109 8.7053>, rot=<0.0355 -0.0009 0.9993 0.0155>}
 calf_r {parent=thigh_r, loc=<42.5284 0.0019 0.7297>, rot=<0.0006 -0.0161 -0.0391 -0.9991>}
 foot_r {parent=calf_r, loc=<46.7401 2.5014 0.592>, rot=<0.0427 -0.0238 0.0353 -0.9982>}
 ball_r {parent=foot_r, loc=<3.5658 -10.8066 0.1544>, rot=<0.0001 -0.0001 -0.7063 -0.7079>}
 calf_twist_02_r {parent=calf_r, loc=<19.1546 1.7943 0.2638>, rot=<0.0237 0.0095 -0.0035 -0.9997>}
 dyn_bows_leg_r_base {parent=calf_twist_02_r, loc=<-0.5968 -5.7698 0.6439>, rot=<0.4872 -0.4851 0.537 -0.4889>}
 dyn_bows_leg_r_01 {parent=dyn_bows_leg_r_base, loc=<0 0.3548 0>, rot=<-0.0264 0.0255 -0.694 -0.719>}
 dyn_bows_leg_r_02 {parent=dyn_bows_leg_r_base, loc=<0 -0.3994 0>, rot=<0.0797 -0.0799 0.6602 -0.7426>}
 calf_twist_01_r {parent=calf_r, loc=<35.8483 1.9197 0.5794>, rot=<0.0237 0.0095 -0.0035 -0.9997>}
 deform_knee_in_base_r {parent=calf_r, loc=<-0.8031 5.0231 -0.2457>, rot=<0 0 0 -1>}
 deform_knee_in_r {parent=deform_knee_in_base_r, loc=<0 0 0>, rot=<0 0 0 1>}
 deform_knee_out_base_r {parent=calf_r, loc=<0 -4 0>, rot=<0 0 0 -1>}
 deform_knee_out_r {parent=deform_knee_out_base_r, loc=<0 0 0>, rot=<0 0 0 1>}
 thigh_twist_01_r {parent=thigh_r, loc=<27.952 0 0>, rot=<0.0013 0 0 -1>}
 deform_groin_base_r {parent=thigh_r, loc=<-0.4801 -5.6776 0.2486>, rot=<0 0 0 -1>}
 deform_groin_r {parent=deform_groin_base_r, loc=<0 0 0>, rot=<0 0 0 1>}
 deform_glute_base_l {parent=pelvis, loc=<-7.3466 -9.0213 -9.1245>, rot=<0.0172 0.031 -0.4477 -0.8935>}
 deform_glute_l {parent=deform_glute_base_l, loc=<0 0 0>, rot=<0 0 0 1>}
 deform_glute_base_r {parent=pelvis, loc=<-7.3464 -9.0213 9.1245>, rot=<0.031 -0.0172 0.8935 -0.4477>}
 deform_glute_r {parent=deform_glute_base_r, loc=<0 0 0>, rot=<0 0 0 1>}
]
