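# Physics asset for the female medium BrainMatter Jacket cosmetic, providing secondary motion for the jacket's two front flaps.
# Three large spine capsules form a torso chain with all angular motion locked.
# Small left and right flap base bodies are locked to spine_05, each anchoring a capsule that swings within tight limits with zero twist range.
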PhysicsAsset /BRCosmetics/Characters/Player/Female/Medium/Bodies/F_MED_BrainMatter_Jacket/Meshes/F_MED_BrainMatter_Jacket_Physics
bodies {
 dyn_jacket_f_l_01 [{shapeType=Capsule, length=5, radius=3, shapeName=dyn_jacket_f_l_01_capsule, center=<9.5706 -2.5 0>, rotation=<rot 0 0 3.4522>}]
 dyn_jacket_f_l_base [{shapeType=Capsule, length=0.505, radius=0.505, shapeName=dyn_jacket_f_l_base_capsule, rotation=<rot 0 -90 0>}]
 dyn_jacket_f_r_01 [{shapeType=Capsule, length=7, radius=3, shapeName=dyn_jacket_f_r_01_capsule, center=<8.5793 2.5 0.0939>, rotation=<rot 0.0001 0.0001 -0.645>}]
 dyn_jacket_f_r_base [{shapeType=Capsule, length=0.505, radius=0.505, shapeName=dyn_jacket_f_r_base_capsule, rotation=<rot 0 -90 0>}]
 spine_01 [{
   shapeType=Capsule
   length=8.6153
   radius=12.688
   shapeName=spine_01_capsule
   center=<-3.6125 -0.2275 0.0272>
   rotation=<rot -0.0271 -86.3974 90.4299>
  }]
 spine_03 [{shapeType=Capsule, length=8.5751, radius=12.3556, shapeName=spine_03_capsule, center=<-0.9735 0 0.0057>, rotation=<rot 0 -90.0001 90.3334>}]
 spine_05 [{
   shapeType=Capsule
   length=11.1655
   radius=10.8804
   shapeName=spine_05_capsule
   center=<4.2427 -0.6743 0.0089>
   rotation=<rot 0.217 102.355 -90>
  }]
}
constraints [
 {bone1Name=dyn_jacket_f_r_01, bone2Name=dyn_jacket_f_r_base, swing1LimitDegrees=10, swing2LimitDegrees=5, twistLimitDegrees=0}
 {
  bone1Name=dyn_jacket_f_r_base
  bone2Name=spine_05
  swing1LimitDegrees=45
  swing1Motion=Locked
  swing2LimitDegrees=45
  swing2Motion=Locked
  twistLimitDegrees=45
  twistMotion=Locked
 }
 {
  bone1Name=dyn_jacket_f_l_base
  bone2Name=spine_05
  swing1LimitDegrees=45
  swing1Motion=Locked
  swing2LimitDegrees=45
  swing2Motion=Locked
  twistLimitDegrees=45
  twistMotion=Locked
 }
 {
  bone1Name=spine_03
  bone2Name=spine_01
  swing1LimitDegrees=45
  swing1Motion=Locked
  swing2LimitDegrees=45
  swing2Motion=Locked
  twistLimitDegrees=45
  twistMotion=Locked
 }
 {
  bone1Name=spine_05
  bone2Name=spine_03
  swing1LimitDegrees=45
  swing1Motion=Locked
  swing2LimitDegrees=45
  swing2Motion=Locked
  twistLimitDegrees=45
  twistMotion=Locked
 }
 {bone1Name=dyn_jacket_f_l_01, bone2Name=dyn_jacket_f_l_base, swing1LimitDegrees=10, swing2LimitDegrees=5, twistLimitDegrees=0}
]
