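# Post-process style AnimBlueprint for the female BrainMatter Jacket body, adding secondary motion to fringe, bows and jacket.
# Copies the parent mesh pose, then adjusts scale and slide-pose bones and constrains back-fringe bases.
# Backbling-dependent RigidBody chains simulate back fringe, with a reduced-alpha branch during the RollerBlade emote.
# Control Rig, AnimDynamics for arm fringe and leg bows, jacket RigidBody and a deform rig follow.
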AnimBlueprint /BRCosmetics/Characters/Player/Female/Medium/Bodies/F_MED_BrainMatter_Jacket/Meshes/F_MED_BrainMatter_Jacket_AnimBP
target_skeleton=&/F_MED_BrainMatter_Jacket_Skeleton
graphs [
 AnimGraph {nodes [
   n1 = RigidBody!(
    ComponentPose=Local To Component(LocalPose=Use cached pose 'Pre sim'())
    bAlphaBoolEnabled:bool=Get Is Backbling A Cape() -> (Is Backbling A Cape:bool)
    OverrideWorldGravity:Vector="0,0,0"
    ExternalForce
   ) -> (Pose) {
    alphaInputType=Bool
    bClampLinearTranslationLimitToRefPose=true
    bOverrideWorldGravity=true
    overridePhysicsAsset=&/F_MED_BrainMatter_Jacket_BackFringe
    simulationSpace=BaseBoneSpace
    alphaBoolBlend {blendOption=Linear}
    baseBoneRef {boneName=spine_05}
    componentAppliedLinearAccClamp=<3000 3000 3000>
    componentLinearAccScale=<0.3 0.3 0.3>
    simSpaceSettings {maxAngularAcceleration=500, maxAngularVelocity=20, maxLinearAcceleration=400, maxLinearVelocity=500, worldAlpha=1}
   }
   n2 = RigidBody(ComponentPose=@n1, bAlphaBoolEnabled=Get IsWearingBackbling() -> (IsWearingBackbling:bool), OverrideWorldGravity="0,0,0") {
    alphaInputType=Bool
    bClampLinearTranslationLimitToRefPose=true
    bOverrideWorldGravity=true
    overridePhysicsAsset=&/F_MED_BrainMatter_Jacket_BackFringe_Backpack
    simulationSpace=BaseBoneSpace
    alphaBoolBlend {blendOption=Linear}
    baseBoneRef {boneName=spine_05}
    componentAppliedLinearAccClamp=<3000 3000 3000>
    componentLinearAccScale=<0.3 0.3 0.3>
    simSpaceSettings {maxAngularAcceleration=500, maxAngularVelocity=20, maxLinearAcceleration=400, maxLinearVelocity=500, worldAlpha=1}
   }
   n3 = RigidBody(ComponentPose=@n2, bAlphaBoolEnabled=Get Is no Backbling() -> (Is no Backbling:bool), OverrideWorldGravity="0,0,0") {
    alphaInputType=Bool
    bClampLinearTranslationLimitToRefPose=true
    bOverrideWorldGravity=true
    overridePhysicsAsset=&F_MED_BrainMatter_Jacket_BackFringe
    simulationSpace=BaseBoneSpace
    alphaBoolBlend {blendOption=Linear}
    baseBoneRef {boneName=spine_05}
    componentAppliedLinearAccClamp=<3000 3000 3000>
    componentLinearAccScale=<0.3 0.3 0.3>
    simSpaceSettings {maxAngularAcceleration=500, maxAngularVelocity=20, maxLinearAcceleration=400, maxLinearVelocity=500, worldAlpha=1}
   }
   n4 = RigidBody!(
    ComponentPose=Local To Component(LocalPose=Use cached pose 'Pre sim'())
    Alpha:float=Get is wear cape emote() -> (is wear cape emote:double)
    OverrideWorldGravity="0,0,0"
    ExternalForce
   ) -> (Pose) {
    bClampLinearTranslationLimitToRefPose=true
    bOverrideWorldGravity=true
    overridePhysicsAsset=&F_MED_BrainMatter_Jacket_BackFringe
    simulationSpace=BaseBoneSpace
    alphaBoolBlend {blendOption=Linear}
    baseBoneRef {boneName=spine_05}
    componentAppliedLinearAccClamp=<3000 3000 3000>
    componentLinearAccScale=<0.3 0.3 0.3>
    simSpaceSettings {maxAngularAcceleration=500, maxAngularVelocity=20, maxLinearAcceleration=400, maxLinearVelocity=500, worldAlpha=1}
   }
   n5 = RigidBody(ComponentPose=@n4, Alpha=Get Is backpack emote() -> (Is backpack emote:double), OverrideWorldGravity="0,0,0") {
    bClampLinearTranslationLimitToRefPose=true
    bOverrideWorldGravity=true
    overridePhysicsAsset=&F_MED_BrainMatter_Jacket_BackFringe_Backpack
    simulationSpace=BaseBoneSpace
    alphaBoolBlend {blendOption=Linear}
    baseBoneRef {boneName=spine_05}
    componentAppliedLinearAccClamp=<3000 3000 3000>
    componentLinearAccScale=<0.3 0.3 0.3>
    simSpaceSettings {maxAngularAcceleration=500, maxAngularVelocity=20, maxLinearAcceleration=400, maxLinearVelocity=500, worldAlpha=1}
   }
   n6 = RigidBody(ComponentPose=@n5, Alpha=Get Is no backpac emote() -> (Is no backpac emote:double), OverrideWorldGravity="0,0,0") {
    bClampLinearTranslationLimitToRefPose=true
    bOverrideWorldGravity=true
    overridePhysicsAsset=&F_MED_BrainMatter_Jacket_BackFringe
    simulationSpace=BaseBoneSpace
    alphaBoolBlend {blendOption=Linear}
    baseBoneRef {boneName=spine_05}
    componentAppliedLinearAccClamp=<3000 3000 3000>
    componentLinearAccScale=<0.3 0.3 0.3>
    simSpaceSettings {maxAngularAcceleration=500, maxAngularVelocity=20, maxLinearAcceleration=400, maxLinearVelocity=500, worldAlpha=1}
   }
   n7 = Blend Poses by bool(
    BlendPose_0=Component To Local(ComponentPose=@n6)
    BlendPose_1=Component To Local(ComponentPose=@n3)
    bActiveValue=Get Emote alpha() -> (Emote alpha:bool)
   ) {blendType=Linear}
   n8 = Control Rig(Source=Component To Local(ComponentPose=Local To Component(LocalPose=@n7))) {
    alphaBoolBlend {blendOption=Linear}
    controlRigAssetReference {blueprintRigClass=&/F_MED_BrainMatter_Jacket_CtrlRig.F_MED_BrainMatter_Jacket_CtrlRig_C}
   }
   n9 = "Anim Dynamics - Chain: dyn_arm_fringe_r_01_01 -> dyn_arm_fringe_r_01_02"(ComponentPose=Local To Component(LocalPose=@n8), Alpha=1, ExternalForce="0,0,0") {
    angularDampingOverride=1
    bChain=true
    bGravityOverrideInSimSpace=true
    bOverrideAngularDamping=true
    bUseGravityOverride=true
    numSolverIterationsPreUpdate=2
    simulationSpace=RootRelative
    alphaBoolBlend {blendOption=Linear}
    alphaScaleBiasClamp {interpSpeedDecreasing=150, interpSpeedIncreasing=150}
    boundBone {boneName=dyn_arm_fringe_r_01_01}
    chainEnd {boneName=dyn_arm_fringe_r_01_02}
    gravityOverride {z=-1000}
    physicsBodyDefinitions [
     {
      boundBone {boneName=dyn_arm_fringe_r_01_01}
      boxExtents=<10 10 10>
      constraintSetup {angularLimitsMax=<0 50 50>, angularLimitsMin=<0 -50 -50>, angularTarget=<0 0 0>}
      localJointOffset=<0 0 0>
     }
     {
      boundBone {boneName=dyn_arm_fringe_r_01_02}
      boxExtents=<10 10 10>
      constraintSetup {angularLimitsMax=<0 50 20>, angularLimitsMin=<0 -50 -20>, angularTarget=<0 0 0>}
      localJointOffset=<0 0 0>
     }
    ]
    relativeSpaceBone {boneName=lowerarm_r}
   }
   n10 = "Anim Dynamics - Chain: dyn_arm_fringe_r_02_01 -> dyn_arm_fringe_r_02_02"(ComponentPose=@n9, Alpha=1, ExternalForce="0,0,0") {
    angularDampingOverride=1
    angularSpringConstant=4000
    bChain=true
    bGravityOverrideInSimSpace=true
    bOverrideAngularDamping=true
    bUseGravityOverride=true
    numSolverIterationsPreUpdate=2
    simulationSpace=RootRelative
    alphaBoolBlend {blendOption=Linear}
    alphaScaleBiasClamp {interpSpeedDecreasing=150, interpSpeedIncreasing=150}
    boundBone {boneName=dyn_arm_fringe_r_02_01}
    chainEnd {boneName=dyn_arm_fringe_r_02_02}
    gravityOverride {z=-1000}
    physicsBodyDefinitions [
     {
      boundBone {boneName=dyn_arm_fringe_r_02_01}
      boxExtents=<10 10 10>
      constraintSetup {angularLimitsMax=<0 50 50>, angularLimitsMin=<0 -50 -50>, angularTarget=<0 0 0>}
      localJointOffset=<0 0 0>
     }
     {
      boundBone {boneName=dyn_arm_fringe_r_02_02}
      boxExtents=<10 10 10>
      constraintSetup {angularLimitsMax=<0 50 20>, angularLimitsMin=<0 -50 -20>, angularTarget=<0 0 0>}
      localJointOffset=<0 0 0>
     }
    ]
    relativeSpaceBone {boneName=lowerarm_r}
   }
   n11 = "Anim Dynamics - Chain: dyn_arm_fringe_r_03_01 -> dyn_arm_fringe_r_03_02"(ComponentPose=@n10, Alpha=1, ExternalForce="0,0,0") {
    angularBiasOverride=2
    angularDampingOverride=1
    angularSpringConstant=4000
    bChain=true
    bGravityOverrideInSimSpace=true
    bOverrideAngularDamping=true
    bUseGravityOverride=true
    numSolverIterationsPreUpdate=2
    simulationSpace=RootRelative
    alphaBoolBlend {blendOption=Linear}
    alphaScaleBiasClamp {interpSpeedDecreasing=150, interpSpeedIncreasing=150}
    boundBone {boneName=dyn_arm_fringe_r_03_01}
    chainEnd {boneName=dyn_arm_fringe_r_03_02}
    gravityOverride {z=-1000}
    physicsBodyDefinitions [
     {
      boundBone {boneName=dyn_arm_fringe_r_03_01}
      boxExtents=<10 10 10>
      constraintSetup {angularLimitsMax=<0 120 50>, angularLimitsMin=<0 -50 -50>, angularTarget=<0 0 0>}
      localJointOffset=<0 0 0>
     }
     {
      boundBone {boneName=dyn_arm_fringe_r_03_02}
      boxExtents=<10 10 10>
      constraintSetup {angularLimitsMax=<0 50 20>, angularLimitsMin=<0 -50 -20>, angularTarget=<0 0 0>}
      localJointOffset=<0 0 0>
     }
    ]
    relativeSpaceBone {boneName=lowerarm_r}
   }
   n12 = "Anim Dynamics - Chain: dyn_arm_fringe_l_01_01 -> dyn_arm_fringe_l_01_02"(ComponentPose=@n11, Alpha=1, ExternalForce="0,0,0") {
    angularDampingOverride=1
    angularSpringConstant=4000
    bChain=true
    bGravityOverrideInSimSpace=true
    bOverrideAngularDamping=true
    bUseGravityOverride=true
    numSolverIterationsPreUpdate=2
    simulationSpace=RootRelative
    alphaBoolBlend {blendOption=Linear}
    alphaScaleBiasClamp {interpSpeedDecreasing=150, interpSpeedIncreasing=150}
    boundBone {boneName=dyn_arm_fringe_l_01_01}
    chainEnd {boneName=dyn_arm_fringe_l_01_02}
    gravityOverride {z=-1000}
    physicsBodyDefinitions [
     {
      boundBone {boneName=dyn_arm_fringe_l_01_01}
      boxExtents=<10 10 10>
      constraintSetup {angularLimitsMax=<0 50 50>, angularLimitsMin=<0 -50 -50>, angularTarget=<0 0 0>}
      localJointOffset=<0 0 0>
     }
     {
      boundBone {boneName=dyn_arm_fringe_l_01_02}
      boxExtents=<10 10 10>
      constraintSetup {angularLimitsMax=<0 50 20>, angularLimitsMin=<0 -50 -20>, angularTarget=<0 0 0>}
      localJointOffset=<0 0 0>
     }
    ]
    relativeSpaceBone {boneName=lowerarm_l}
   }
   n13 = "Anim Dynamics - Chain: dyn_arm_fringe_l_02_01 -> dyn_arm_fringe_l_02_02"(ComponentPose=@n12, Alpha=1, ExternalForce="0,0,0") {
    angularDampingOverride=1
    angularSpringConstant=4000
    bChain=true
    bGravityOverrideInSimSpace=true
    bOverrideAngularDamping=true
    bUseGravityOverride=true
    numSolverIterationsPreUpdate=2
    simulationSpace=RootRelative
    alphaBoolBlend {blendOption=Linear}
    alphaScaleBiasClamp {interpSpeedDecreasing=150, interpSpeedIncreasing=150}
    boundBone {boneName=dyn_arm_fringe_l_02_01}
    chainEnd {boneName=dyn_arm_fringe_l_02_02}
    gravityOverride {z=-1000}
    physicsBodyDefinitions [
     {
      boundBone {boneName=dyn_arm_fringe_l_02_01}
      boxExtents=<10 10 10>
      constraintSetup {angularLimitsMax=<0 50 50>, angularLimitsMin=<0 -50 -50>, angularTarget=<0 0 0>}
      localJointOffset=<0 0 0>
     }
     {
      boundBone {boneName=dyn_arm_fringe_l_02_02}
      boxExtents=<10 10 10>
      constraintSetup {angularLimitsMax=<0 50 20>, angularLimitsMin=<0 -50 -20>, angularTarget=<0 0 0>}
      localJointOffset=<0 0 0>
     }
    ]
    relativeSpaceBone {boneName=lowerarm_l}
   }
   n14 = "Anim Dynamics - Chain: dyn_arm_fringe_l_03_02 -> dyn_arm_fringe_l_03_01"(ComponentPose=@n13, Alpha=1, ExternalForce="0,0,0") {
    angularDampingOverride=1
    angularSpringConstant=4000
    bChain=true
    bGravityOverrideInSimSpace=true
    bOverrideAngularDamping=true
    bUseGravityOverride=true
    numSolverIterationsPreUpdate=2
    simulationSpace=RootRelative
    alphaBoolBlend {blendOption=Linear}
    alphaScaleBiasClamp {interpSpeedDecreasing=150, interpSpeedIncreasing=150}
    boundBone {boneName=dyn_arm_fringe_l_03_02}
    chainEnd {boneName=dyn_arm_fringe_l_03_01}
    gravityOverride {z=-1000}
    physicsBodyDefinitions [
     {
      boundBone {boneName=dyn_arm_fringe_l_03_02}
      boxExtents=<10 10 10>
      constraintSetup {angularLimitsMax=<0 80 50>, angularLimitsMin=<0 -50 -50>, angularTarget=<0 0 0>}
      localJointOffset=<0 0 0>
     }
     {
      boundBone {boneName=dyn_arm_fringe_l_03_01}
      boxExtents=<10 10 10>
      constraintSetup {angularLimitsMax=<0 50 20>, angularLimitsMin=<0 -50 -20>, angularTarget=<0 0 0>}
      localJointOffset=<0 0 0>
     }
    ]
    relativeSpaceBone {boneName=lowerarm_twist_02_l}
   }
   n15 = Anim Dynamics - Bone: dyn_bows_leg_l_01(ComponentPose=@n14, Alpha=1, ExternalForce="0,0,0") {
    angularDampingOverride=0.9995
    angularSpringConstant=2500
    bAngularSpring=true
    bGravityOverrideInSimSpace=true
    bOverrideAngularDamping=true
    bUseGravityOverride=true
    alphaBoolBlend {blendOption=Linear}
    boundBone {boneName=dyn_bows_leg_l_01}
    componentAppliedLinearAccClamp=<3000 3000 3000>
    componentLinearAccScale=<0.3 0.6 0.3>
    componentLinearVelScale=<0.3 0.3 0.3>
    gravityOverride {z=-980}
    physicsBodyDefinitions [{
      boundBone {boneName=dyn_bows_leg_l_01}
      boxExtents=<10 10 10>
      constraintSetup {angularTargetAxis=AxisZ, angularLimitsMax=<0 50 0>, angularLimitsMin=<0 -40 0>, angularTarget=<0 0 1>}
      localJointOffset=<0 0 0>
     }]
   }
   n16 = Anim Dynamics - Bone: dyn_bows_leg_l_02(ComponentPose=@n15, Alpha=1, ExternalForce="0,0,0") {
    angularDampingOverride=0.9995
    angularSpringConstant=2500
    bAngularSpring=true
    bGravityOverrideInSimSpace=true
    bOverrideAngularDamping=true
    bUseGravityOverride=true
    alphaBoolBlend {blendOption=Linear}
    boundBone {boneName=dyn_bows_leg_l_02}
    componentAppliedLinearAccClamp=<3000 3000 3000>
    componentLinearAccScale=<0.3 0.6 0.3>
    componentLinearVelScale=<0.3 0.3 0.3>
    gravityOverride {z=-980}
    physicsBodyDefinitions [{
      boundBone {boneName=dyn_bows_leg_l_02}
      boxExtents=<10 10 10>
      constraintSetup {angularTargetAxis=AxisZ, angularLimitsMax=<0 50 0>, angularLimitsMin=<0 -40 0>, angularTarget=<0 0 1>}
      localJointOffset=<0 0 0>
     }]
   }
   n17 = Anim Dynamics - Bone: dyn_bows_leg_r_01(ComponentPose=@n16, Alpha=1, ExternalForce="0,0,0") {
    angularDampingOverride=0.9995
    angularSpringConstant=2500
    bAngularSpring=true
    bGravityOverrideInSimSpace=true
    bOverrideAngularDamping=true
    bUseGravityOverride=true
    alphaBoolBlend {blendOption=Linear}
    boundBone {boneName=dyn_bows_leg_r_01}
    componentAppliedLinearAccClamp=<3000 3000 3000>
    componentLinearAccScale=<0.3 0.6 0.3>
    componentLinearVelScale=<0.3 0.3 0.3>
    gravityOverride {z=-980}
    physicsBodyDefinitions [{
      boundBone {boneName=dyn_bows_leg_r_01}
      boxExtents=<10 10 10>
      constraintSetup {angularTargetAxis=AxisZ, angularLimitsMax=<0 50 0>, angularLimitsMin=<0 -40 0>, angularTarget=<0 0 1>}
      localJointOffset=<0 0 0>
     }]
   }
   n18 = Anim Dynamics - Bone: dyn_bows_leg_r_02(ComponentPose=@n17, Alpha=1, ExternalForce="0,0,0") {
    angularDampingOverride=0.9995
    angularSpringConstant=2500
    bAngularSpring=true
    bGravityOverrideInSimSpace=true
    bOverrideAngularDamping=true
    bUseGravityOverride=true
    alphaBoolBlend {blendOption=Linear}
    boundBone {boneName=dyn_bows_leg_r_02}
    componentAppliedLinearAccClamp=<3000 3000 3000>
    componentLinearAccScale=<0.3 0.6 0.3>
    componentLinearVelScale=<0.3 0.3 0.3>
    gravityOverride {z=-980}
    physicsBodyDefinitions [{
      boundBone {boneName=dyn_bows_leg_r_02}
      boxExtents=<10 10 10>
      constraintSetup {angularTargetAxis=AxisZ, angularLimitsMax=<0 50 0>, angularLimitsMin=<0 -40 0>, angularTarget=<0 0 1>}
      localJointOffset=<0 0 0>
     }]
   }
   n19 = RigidBody(ComponentPose=@n18, OverrideWorldGravity="0,0,0") {
    bClampLinearTranslationLimitToRefPose=true
    bOverrideWorldGravity=true
    overridePhysicsAsset=&/F_MED_BrainMatter_Jacket_Physics
    simulationSpace=BaseBoneSpace
    alphaBoolBlend {blendOption=Linear}
    baseBoneRef {boneName=spine_05}
    componentAppliedLinearAccClamp=<3000 3000 3000>
    simSpaceSettings {
     maxAngularAcceleration=500
     maxAngularVelocity=3
     maxLinearAcceleration=400
     maxLinearVelocity=250
     velocityScaleZ=0
     worldAlpha=1
     externalLinearDragV=<0.1 0.1>
    }
   }
   n20 = Control Rig(Source=Component To Local(ComponentPose=@n19)) {
    alphaBoolBlend {blendOption=Linear}
    controlRigAssetReference {blueprintRigClass=&/F_MED_BrainMatter_Jacket_DeformRig.F_MED_BrainMatter_Jacket_DeformRig_C}
   }
   Output Pose(Result=@n20) {layerGroup=None}
   n21 = "Transform (Modify) Bone - Bone: scale_shouler_l"(
    ComponentPose=Local To Component(LocalPose=Copy Pose From Mesh() {bUseAttachedParent=true})
    Alpha=1
    Translation="0,0,0"
    Rotation="0,0,0"
    Scale="0,0,0"
   ) {
    scaleMode=BMM_Additive
    scaleSpace=BCS_BoneSpace
    alphaBoolBlend {blendOption=Linear}
    boneToModify {boneName=scale_shouler_l}
    scale=<0 0 0>
   }
   n22 = "Transform (Modify) Bone - Bone: scale_shouler_r"(ComponentPose=@n21, Alpha=1, Translation="2.738,1.197,-0.676", Rotation="0,0,0", Scale="0,0,0") {
    scaleMode=BMM_Additive
    scaleSpace=BCS_BoneSpace
    translationMode=BMM_Additive
    alphaBoolBlend {blendOption=Linear}
    boneToModify {boneName=scale_shouler_r}
    scale=<0 0 0>
    translation=<2.738 1.197 -0.676>
   }
   n23 = "Transform (Modify) Bone - Bone: Scale_neck"(ComponentPose=@n22, Alpha=1, Translation="0,3,0", Rotation="0,0,0", Scale="1,1,0.8") {
    scaleMode=BMM_Additive
    scaleSpace=BCS_BoneSpace
    translationMode=BMM_Additive
    translationSpace=BCS_BoneSpace
    alphaBoolBlend {blendOption=Linear}
    boneToModify {boneName=Scale_neck}
    scale {z=0.8}
    translation {y=3}
   }
   n24 = "Transform (Modify) Bone - Bone: Scale_spine"(ComponentPose=@n23, Alpha=Get Is LOD 1() -> (Is LOD 1:double), Translation="0,3,0", Rotation="0,0,0", Scale="1,1,1") {
    scaleSpace=BCS_BoneSpace
    translationMode=BMM_Additive
    translationSpace=BCS_BoneSpace
    alphaBoolBlend {blendOption=Linear}
    boneToModify {boneName=Scale_spine}
    translation {y=3}
   }
   n25 = "Transform (Modify) Bone - Bone: scale_spine_body"(ComponentPose=@n24, Alpha=1, Translation="0.317,0.498,-0.000", Rotation="0,0,0", Scale="0.7,1,1") {
    scaleMode=BMM_Additive
    scaleSpace=BCS_BoneSpace
    translationMode=BMM_Additive
    translationSpace=BCS_BoneSpace
    alphaBoolBlend {blendOption=Linear}
    boneToModify {boneName=scale_spine_body}
    scale {x=0.7}
    translation=<0.317 0.498>
   }
   n26 = Get bIsKneeSliding()
   n27 = "Transform (Modify) Bone - Bone: dyn_jacket_f_r_base"!(ComponentPose=@n25, bAlphaBoolEnabled:bool=@n26, Translation="0.123,0.987,-0.187", Rotation="0,0,0", Scale="1,1,1") -> (Pose) {
    alphaInputType=Bool
    rotationMode=BMM_Additive
    rotationSpace=BCS_BoneSpace
    translationMode=BMM_Additive
    translationSpace=BCS_BoneSpace
    alphaBoolBlend {blendInTime=0.5, blendOption=Linear, blendOutTime=0.5}
    boneToModify {boneName=dyn_jacket_f_r_base}
    translation=<0.123 0.987 -0.187>
   }
   n28 = "Transform (Modify) Bone - Bone: dyn_jacket_f_l_base"!(ComponentPose=@n27, bAlphaBoolEnabled:bool=@n26, Translation="-2.372,0.864,-1.399", Rotation="0,0,0", Scale="1,1,1") -> (Pose) {
    alphaInputType=Bool
    rotationMode=BMM_Additive
    rotationSpace=BCS_BoneSpace
    translationMode=BMM_Additive
    translationSpace=BCS_BoneSpace
    alphaBoolBlend {blendInTime=0.5, blendOption=Linear, blendOutTime=0.5}
    boneToModify {boneName=dyn_jacket_f_l_base}
    translation=<-2.372 0.864 -1.399>
   }
   n29 = Get bIsBaseballSliding()
   n30 = "Transform (Modify) Bone - Bone: dyn_jacket_f_r_base"(ComponentPose=@n28, bAlphaBoolEnabled=@n29, Translation="-2.757,3.439,0.099", Rotation="11.634,0.194,1.950", Scale="1,1,1") {
    alphaInputType=Bool
    rotationMode=BMM_Additive
    rotationSpace=BCS_BoneSpace
    translationMode=BMM_Additive
    translationSpace=BCS_BoneSpace
    alphaBoolBlend {blendInTime=0.5, blendOption=Linear, blendOutTime=0.5}
    boneToModify {boneName=dyn_jacket_f_r_base}
    rotation=<rot 11.6341 0.194 1.95>
    translation=<-2.757 3.439 0.099>
   }
   n31 = "Transform (Modify) Bone - Bone: dyn_jacket_f_l_base"(ComponentPose=@n30, bAlphaBoolEnabled=@n29, Translation="-3.340,-3.033,-8.618", Rotation="-47.095,-48.425,-59.023", Scale="1,1,1") {
    alphaInputType=Bool
    rotationMode=BMM_Additive
    rotationSpace=BCS_BoneSpace
    translationMode=BMM_Additive
    translationSpace=BCS_BoneSpace
    alphaBoolBlend {blendInTime=0.5, blendOption=Linear, blendOutTime=0.5}
    boneToModify {boneName=dyn_jacket_f_l_base}
    rotation=<rot -47.095 -48.425 -59.023>
    translation=<-3.34 -3.033 -8.618>
   }
   n32 = "Constraint - dyn_back_fringe_01_Base (PPPPPPPP)"!(
    ComponentPose=@n31
    Alpha=1
    ConstraintWeights_0=0.0026
    ConstraintWeights_1:float=0.165
    ConstraintWeights_2:float=0.4631
    ConstraintWeights_3:float=0.2268
    ConstraintWeights_4:float=0.0902
    ConstraintWeights_5:float=0.0026
    ConstraintWeights_6:float=0.0447
    ConstraintWeights_7:float=0.005
   ) -> (Pose) {
    alphaBoolBlend {blendOption=Linear}
    boneToModify {boneName=dyn_back_fringe_01_Base}
    constraintSetup [
     {targetBone {boneName=upperarm_twist_02_l}}
     {targetBone {boneName=upperarm_twist_01_l}}
     {targetBone {boneName=spine_04}}
     {targetBone {boneName=spine_05}}
     {targetBone {boneName=clavicle_l}}
     {targetBone {boneName=clavicle_r}}
     {targetBone {boneName=deform_upperarm_bk_l}}
     {targetBone {boneName=deform_upperarm_up_l}}
    ]
    constraintWeights [1,1,1,1,1,1,1,1]
   }
   n33 = "Constraint - dyn_back_fringe_02_base (PPPPPPPP)"!(
    ComponentPose=@n32
    Alpha=1
    ConstraintWeights_0=0.0401
    ConstraintWeights_1:float=0.4817
    ConstraintWeights_2:float=0.302
    ConstraintWeights_3:float=0.1121
    ConstraintWeights_4:float=0.0246
    ConstraintWeights_5:float=0.0166
    ConstraintWeights_6:float=0.0025
    ConstraintWeights_7:float=0.0204
   ) -> (Pose) {
    alphaBoolBlend {blendOption=Linear}
    boneToModify {boneName=dyn_back_fringe_02_base}
    constraintSetup [
     {targetBone {boneName=upperarm_twist_01_l}}
     {targetBone {boneName=spine_04}}
     {targetBone {boneName=spine_05}}
     {targetBone {boneName=clavicle_l}}
     {targetBone {boneName=clavicle_r}}
     {targetBone {boneName=deform_upperarm_bk_l}}
     {targetBone {boneName=deform_upperarm_up_l}}
     {targetBone {boneName=dyn_collar_r_02}}
    ]
    constraintWeights [1,1,1,1,1,1,1,1]
   }
   n34 = "Constraint - dyn_back_fringe_03_base (PPPPPPPP)"!(
    ComponentPose=@n33
    Alpha=1
    ConstraintWeights_0=0.0028
    ConstraintWeights_1:float=0.5979
    ConstraintWeights_2:float=0.2066
    ConstraintWeights_3:float=0.0875
    ConstraintWeights_4:float=0.081
    ConstraintWeights_5:float=0.0038
    ConstraintWeights_6:float=0.0169
    ConstraintWeights_7:float=0.0036
   ) -> (Pose) {
    alphaBoolBlend {blendOption=Linear}
    boneToModify {boneName=dyn_back_fringe_03_base}
    constraintSetup [
     {targetBone {boneName=upperarm_twist_01_l}}
     {targetBone {boneName=spine_04}}
     {targetBone {boneName=spine_05}}
     {targetBone {boneName=clavicle_l}}
     {targetBone {boneName=clavicle_r}}
     {targetBone {boneName=upperarm_twist_01_r}}
     {targetBone {boneName=dyn_collar_l_02}}
     {targetBone {boneName=dyn_collar_r_02}}
    ]
    constraintWeights [1,1,1,1,1,1,1,1]
   }
   n35 = "Constraint - dyn_back_fringe_04_base (PPPPPPP)"!(
    ComponentPose=@n34
    Alpha=1
    ConstraintWeights_0=0.4869
    ConstraintWeights_1:float=0.3091
    ConstraintWeights_2:float=0.0245
    ConstraintWeights_3:float=0.1136
    ConstraintWeights_4:float=0.0289
    ConstraintWeights_5:float=0.0166
    ConstraintWeights_6:float=0.0204
   ) -> (Pose) {
    alphaBoolBlend {blendOption=Linear}
    boneToModify {boneName=dyn_back_fringe_04_base}
    constraintSetup [
     {targetBone {boneName=spine_04}}
     {targetBone {boneName=spine_05}}
     {targetBone {boneName=clavicle_l}}
     {targetBone {boneName=clavicle_r}}
     {targetBone {boneName=upperarm_twist_01_r}}
     {targetBone {boneName=deform_upperarm_bk_r}}
     {targetBone {boneName=dyn_collar_l_02}}
    ]
    constraintWeights [1,1,1,1,1,1,1]
   }
   n36 = "Constraint - dyn_back_fringe_5_base (PPPPPPP)"!(
    ComponentPose=@n35
    Alpha=1
    ConstraintWeights_0=0.4687
    ConstraintWeights_1:float=0.2349
    ConstraintWeights_2:float=0.0023
    ConstraintWeights_3:float=0.0957
    ConstraintWeights_4:float=0.0025
    ConstraintWeights_5:float=0.1473
    ConstraintWeights_6:float=0.0486
   ) -> (Pose) {
    alphaBoolBlend {blendOption=Linear}
    boneToModify {boneName=dyn_back_fringe_5_base}
    constraintSetup [
     {targetBone {boneName=spine_04}}
     {targetBone {boneName=spine_05}}
     {targetBone {boneName=clavicle_l}}
     {targetBone {boneName=clavicle_r}}
     {targetBone {boneName=upperarm_twist_02_r}}
     {targetBone {boneName=upperarm_twist_01_r}}
     {targetBone {boneName=deform_upperarm_bk_r}}
    ]
    constraintWeights [1,1,1,1,1,1,1]
   }
   Save cached pose 'Pre sim'(Pose=Component To Local(ComponentPose=@n36)) {cacheName=Pre sim}
  ]}
 EventGraph {nodes [
   Event BlueprintUpdateAnimation()
   > n1 = Gravity Override State Function DEPRECATED(CharacterPartAnimInstance=Self-Reference(), GameTimeInSeconds=GetGameTimeInSeconds(), GravityOverrideParams=Get Gravity Override Params())
   > Set Gravity Override(Gravity Override=@n1)
   n2 = Get CurrentLOD()
   > Set Is LOD 1(Is LOD 1:double="To Float (Boolean)"(InBool="Equal (Integer)"(A=@n2, B=1))) -> (Output_Get:double)
   n3 = GetAnimInstance(self=GetOwnerSkeletalMeshComponent())
   > Set Emote alpha(Emote alpha:bool="Equal (Object)"(
     A=GetCurrentActiveMontage!(self:Anim Instance Object Reference=@n3) -> (ReturnValue:Anim Montage Object Reference)
     B=/BRCosmetics/Animation/Game/MainPlayer/Emotes/RollerBlade/Emote_RollerBlade_CMF_M
    )) -> (Output_Get:bool)
   > Set Is backpack emote(Is backpack emote:double=SelectFloat(A=0.4, B=0, bPickA=Get IsWearingBackbling())) -> (Output_Get:double)
   > Set is wear cape emote(is wear cape emote:double=SelectFloat(A=0.5, B=0, bPickA=Get Is Backbling A Cape())) -> (Output_Get:double)
   > Set Is no backpac emote(Is no backpac emote:double=SelectFloat(A=0.5, B=0, bPickA=Get Is no Backbling())) -> (Output_Get:double)
   n4 = Event BlueprintInitializeAnimation()
   n5 = CheckForBackbling()
   Set IsWearingBackbling(execute=[@n4.then, @n6.then], IsWearingBackbling:bool=@n5) -> (Output_Get:bool)
   > Set Is Backbling A Cape(Is Backbling A Cape:bool=AND Boolean(A=@n5, B=@n5.bIsBackblingLong)) -> (Output_Get:bool)
   > Set Is no Backbling(Is no Backbling:bool=AND Boolean!(A=NOT Boolean(A=@n5), B=NOT Boolean(A=@n5.bIsBackblingACape), C:bool=NOT Boolean(A=@n5.bIsBackblingLong)) -> (ReturnValue)) -> (Output_Get:bool)
   n6 = Event OnCharacterCustomizationLoaded()
   AND Boolean!(A=false, B=false) -> (ReturnValue)
  ]}
]
variables [
 Emote alpha {type=boolean}
 Gravity Override {type=Vector}
 Gravity Override Params {type=GravityOverrideParamsStruct}
 Is Backbling A Cape {type=boolean}
 Is backpack emote {type=number}
 Is LOD 1 {type=number}
 Is no Backbling {type=boolean}
 Is no backpac emote {type=number}
 is wear cape emote {type=number}
 IsShortBackblingOnAlpha {type=number}
 IsWearingBackbling {type=boolean}
 NotBackblingOrBackblingIsCape {type=number}
]
